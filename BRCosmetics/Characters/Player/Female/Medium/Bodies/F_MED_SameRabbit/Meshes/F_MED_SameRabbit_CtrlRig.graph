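# Control Rig for the F_MED_SameRabbit female cosmetic body that drives corrective deformation bones for shoulder straps and the skirt.
# After a shared DeformRig function, parent and rotation constraints anchor strap bases, then upperarm spherical pose readers weight additive offsets per pose.
# Some left-arm offsets mirror right-arm variables through a seven-transform mirror helper.
# Thigh and calf angles, remapped and multiplied, weight skirt bone corrections for deep leg bends.
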ControlRigBlueprint /BRCosmetics/Characters/Player/Female/Medium/Bodies/F_MED_SameRabbit/Meshes/F_MED_SameRabbit_CtrlRig
graphs [
 New Function_ContainedGraph {nodes [
   n1 = RigVMFunctionEntryNode()
   > Get Mirror Transform = RigVMFunctionReferenceNode(transform=@n1.transform)
   > Get Mirror Transform_1 = RigVMFunctionReferenceNode(transform=@n1.transform_1)
   > Get Mirror Transform_2 = RigVMFunctionReferenceNode(transform=@n1.transform_2)
   > Get Mirror Transform_3 = RigVMFunctionReferenceNode(transform=@n1.transform_3)
   > Get Mirror Transform_4 = RigVMFunctionReferenceNode(transform=@n1.transform_4)
   > Get Mirror Transform_5 = RigVMFunctionReferenceNode(transform=@n1.transform_5)
   > Get Mirror Transform_6 = RigVMFunctionReferenceNode(transform=@n1.transform_6)
   > RigVMFunctionReturnNode(
    mirrorTransform=@"Get Mirror Transform".mirrorTransform
    mirrorTransform_1=@"Get Mirror Transform_1".mirrorTransform
    mirrorTransform_2=@"Get Mirror Transform_2".mirrorTransform
    mirrorTransform_3=@"Get Mirror Transform_3".mirrorTransform
    mirrorTransform_4=@"Get Mirror Transform_4".mirrorTransform
    mirrorTransform_5=@"Get Mirror Transform_5".mirrorTransform
    mirrorTransform_6=@"Get Mirror Transform_6".mirrorTransform
   )
  ]}
 RigVMFunctionLibrary {nodes [Mirror_7_transforms = RigVMCollapseNode(
    ExecuteContext=()
    transform=(Rotation=(X=0.4032,Y=-0.1588,Z=-0.0477,W=0.9),Translation=(X=1.6153,Y=-1.7803,Z=-1.0827),Scale3D=(X=0.813,Y=1,Z=1))
    transform_1=(Rotation=(X=-0.2075,Y=-0.084,Z=-0.4802,W=0.8481),Translation=(X=1.4019,Y=-2.0829,Z=0.9133),Scale3D=(X=1,Y=1,Z=1))
    transform_2=(Rotation=(X=-0.1379,Y=0.0268,Z=-0.014,W=0.99),Translation=(X=0,Y=0,Z=0),Scale3D=(X=1,Y=1,Z=1))
    transform_3=(Rotation=(X=0,Y=0,Z=0,W=1),Translation=(X=0,Y=0,Z=0),Scale3D=(X=1,Y=1,Z=1))
    transform_4=(Rotation=(X=0,Y=0,Z=0,W=1),Translation=(X=0,Y=0,Z=0),Scale3D=(X=1,Y=1,Z=1))
    transform_5=(Rotation=(X=0,Y=0,Z=0,W=1),Translation=(X=0,Y=0,Z=0),Scale3D=(X=1,Y=1,Z=1))
    transform_6=(Rotation=(X=0,Y=0,Z=0,W=1),Translation=(X=0,Y=0,Z=0),Scale3D=(X=1,Y=1,Z=1))
   )]}
 RigVMFunction_Sequence_1_2_ContainedGraph {nodes [
   n1 = RigVMFunctionEntryNode()
   n2 = &RigVMFunction_Sequence(ExecuteContext=@n1.ExecuteContext)
   n3 = &RigVMFunction_Sequence(ExecuteContext=@n2.B)
   RigVMFunctionReturnNode(A=@n2.A, B=@n3.A, C=@n3.B)
  ]}
 RigVMModel {nodes [
   RigUnit_BeginExecution = &RigUnit_BeginExecution()
   FN_DeformRig_FNC = RigVMFunctionReferenceNode(
    DeformRig_ Curve="(deformRig_l_upperarm_up_alpha_72_#1=False,deformRig_r_upperarm_up_alpha_73_#2=false,deformRig_l_upperarm_fr_alpha_74_#3=False,deformRig_r_upperarm_fr_alpha_75_CB51B8 ...(truncated)"
    DeformRig_Bool="(upperarm_1_#4=True,elbow_3_#5=false,wrist_7_#6=false,groin_8_#7=false,glute_17_#8=false,knee_9_40C408324C ...(truncated)"
    DeformRig_Debug=(upperarm_debug_18_#4=false,wrist_debug_20_#6=false,lat_debug_26_#9=false,groinGlute_debug_27_#7=false)
    DeformRig_Value="(upperarm_up_18_#10=(Rotation=(X=0.0,Y=-0.0,Z=0.0,W=1.0),Translation=(X=0.0,Y=0.0,Z=8.0),Scale3D=(X=0.70,Y=1.0,Z=1.0)),upperarm_fwd_fr_19_#11=(Rotation=(X=0.0,Y=-0.0,Z=0.0,W=1.0),Translation=(X=0.0, ...(truncated)"
    ExecuteContext=@RigUnit_BeginExecution.ExecutePin
   )
   > RigVMFunction_Sequence_1 = RigVMAggregateNode()
   n1 = &RigVMFunction_Sequence(ExecuteContext=@RigVMFunction_Sequence_1.A)
   &RigUnit_ParentConstraint(
    AdvancedSettings=(InterpolationType=Average,RotationOrderForFilter=XZY)
    Child=(Type=Bone,Name="deform_strap_front_r_master")
    Parents=((Item=(Type=Bone,Name="clavicle_r"),Weight=0.8),(Item=(Type=Bone,Name="spine_05"),Weight=0.2))
    Weight=1
    ExecutePin=@n1.A
   )
   > &RigUnit_ParentConstraint(
    AdvancedSettings=(InterpolationType=Average,RotationOrderForFilter=XZY)
    Child=(Type=Bone,Name="deform_strap_back_r_master")
    Parents=((Item=(Type=Bone,Name="clavicle_r"),Weight=0.8),(Item=(Type=Bone,Name="spine_05"),Weight=0.2))
    Weight=1
   )
   > &RigUnit_ParentConstraint(
    AdvancedSettings=(InterpolationType=Average,RotationOrderForFilter=XZY)
    Child=(Type=Bone,Name="deform_arm_fr_base_r")
    Parents=((Item=(Type=Bone,Name="clavicle_r"),Weight=0.8),(Item=(Type=Bone,Name="spine_05"),Weight=0.2))
    Weight=1
   )
   > &RigUnit_ParentConstraint(
    AdvancedSettings=(InterpolationType=Average,RotationOrderForFilter=XZY)
    Child=(Type=Bone,Name="deform_arm_bk_base_r")
    Parents=((Item=(Type=Bone,Name="clavicle_r"),Weight=0.9),(Item=(Type=Bone,Name="spine_05"),Weight=0.1))
    Weight=1
   )
   > &RigUnit_RotationConstraintLocalSpaceOffset(
    AdvancedSettings=(InterpolationType=Average,RotationOrderForFilter=ZYX)
    Child=(Type=Bone,Name="deform_upperarmStrap_base_r")
    Filter=(bX=True,bY=false,bZ=false)
    Parents=((Item=(Type=Bone,Name="upperarm_twist_01_r"),Weight=1))
    Weight=1
   )
   > &RigUnit_ParentConstraint(
    AdvancedSettings=(InterpolationType=Average,RotationOrderForFilter=XZY)
    Child=(Type=Bone,Name="deform_upperarmStrap_up_base_r")
    Parents=((Item=(Type=Bone,Name="clavicle_r"),Weight=0.8),(Item=(Type=Bone,Name="upperarm_r"),Weight=0.2))
    Weight=1
   )
   > n2 = &RigUnit_SphericalPoseReader(
    ActiveRegionScaleFactors=(PositiveWidth=1,NegativeWidth=1,PositiveHeight=1,NegativeHeight=1)
    ActiveRegionSize=0.8
    Debug=(bDrawDebug=false,bDraw2D=False,bDrawLocalAxes=False,DebugScale=25,DebugSegments=20,DebugThickness=0.25)
    DriverAxis=(X=-1,Y=0,Z=0)
    DriverItem=(Type=Bone,Name="upperarm_r")
    FalloffRegionScaleFactors=(PositiveWidth=1,NegativeWidth=1,PositiveHeight=1,NegativeHeight=1)
    FalloffSize=0.99
    FlipHeightScaling=False
    FlipWidthScaling=False
    OptionalParentItem=(Type=Bone,Name="clavicle_r")
    RotationOffset=(X=0,Y=-90,Z=0)
   )
   > Get Mirror Transform_2 = RigVMFunctionReferenceNode(transform=(Rotation=(X=-0.0419,Y=0.0513,Z=0.0022,W=0.9978),Translation=(X=0,Y=0,Z=0),Scale3D=(X=1,Y=1,Z=1)))
   > Get Mirror Transform_1_1 = RigVMFunctionReferenceNode(transform=(Rotation=(X=0,Y=0,Z=0,W=1),Translation=(X=-4.2296,Y=5.8904,Z=0),Scale3D=(X=1,Y=1,Z=1)))
   > &RigUnit_ModifyTransforms(
    ItemToModify="((Item=(Type=Bone,Name=\"deform_upperarmStrap_up_r\"),Transform=(Rotation=(X=0,Y=0,Z=0,W=1),Translation=(X=0,Y=0,Z=-2),Scale3D=(X=1,Y=1,Z=1))),(Item=(Type=Bone,Name=\"deform_strap_front_r_01\"),Transform=(Rotation=(X=0,Y=0,Z=0,W=1),Translation=(X=0,Y=0,Z=0),Sc ...(truncated)"
    WeightMinimum=0
    ItemToModify.6.Transform=@"Get Mirror Transform_1_1".mirrorTransform
    ItemToModify.2.Transform=@"Get Mirror Transform_2".mirrorTransform
    Weight=@n2.OutputParam
   )
   > n3 = &RigUnit_SphericalPoseReader(
    ActiveRegionScaleFactors=(PositiveWidth=2.5,NegativeWidth=3,PositiveHeight=3,NegativeHeight=3)
    ActiveRegionSize=0.1
    Debug=(bDrawDebug=false,DebugScale=25,DebugSegments=20,DebugThickness=0.25)
    DriverAxis=(X=-1,Y=0,Z=0)
    DriverItem=(Type=Bone,Name="upperarm_r")
    FalloffRegionScaleFactors=(PositiveWidth=1,NegativeWidth=1,PositiveHeight=1,NegativeHeight=1)
    FalloffSize=0.5
    FlipHeightScaling=true
    OptionalParentItem=(Type=Bone,Name="spine_05")
    RotationOffset=(X=180,Y=0,Z=180)
   )
   > &RigUnit_ModifyTransforms(
    ItemToModify="((Item=(Type=Bone,Name=\"deform_upperarmStrap_up_r\"),Transform=(Rotation=(X=-0.1801,Y=-0.442,Z=0.0414,W=0.8778),Translation=(X=7.1329,Y=-1.4356,Z=-2.8444),Scale3D=(X=1,Y=1,Z=1))),(Item=(Type=Bone,Name=\"deform_strap_front_r_01\"),Transform=(Rotation=(X=0,Y=0, ...(truncated)"
    WeightMinimum=0
    Weight=@n3.OutputParam
   )
   > n4 = &RigUnit_SphericalPoseReader(
    ActiveRegionScaleFactors=(PositiveWidth=0.5,NegativeWidth=3,PositiveHeight=1.5,NegativeHeight=1.5)
    ActiveRegionSize=0.1
    Debug=(bDrawDebug=false,DebugScale=25,DebugSegments=20,DebugThickness=0.25)
    DriverAxis=(X=-1,Y=0,Z=0)
    DriverItem=(Type=Bone,Name="upperarm_r")
    FalloffRegionScaleFactors=(PositiveWidth=1,NegativeWidth=1,PositiveHeight=1,NegativeHeight=1)
    FalloffSize=0.4
    FlipWidthScaling=true
    OptionalParentItem=(Type=Bone,Name="spine_05")
    RotationOffset=(X=90,Y=180,Z=180)
   )
   n5 = RigVMVariableNode() {variable=upperarmFront_L_0}
   n6 = RigVMVariableNode() {variable=upperarmFront_L_2}
   n7 = RigVMVariableNode() {variable=upperarmFront_L_3}
   n8 = RigVMVariableNode() {variable=upperarmFront_L_4}
   n9 = RigVMVariableNode() {variable=upperarmFront_L_6}
   > &RigUnit_ModifyTransforms(
    ItemToModify="((Item=(Type=Bone,Name=\"deform_upperarmStrap_up_r\"),Transform=(Rotation=(X=0.2411,Y=0,Z=0,W=0.9705),Translation=(X=3.4484,Y=-1.0999,Z=0.1149),Scale3D=(X=1,Y=1,Z=1))),(Item=(Type=Bone,Name=\"deform_strap_front_r_01\"),Transform=(Rotation=(X=0.3614,Y=-0.4966,Z ...(truncated)"
    WeightMinimum=0
    Weight=@n4.OutputParam
    ItemToModify.0.Transform=@n5.Value
    ItemToModify.2.Transform=@n6.Value
    ItemToModify.3.Transform=@n7.Value
    ItemToModify.4.Transform=@n8.Value
    ItemToModify.6.Transform=@n9.Value
   )
   > n10 = &RigUnit_SphericalPoseReader(
    ActiveRegionScaleFactors=(PositiveWidth=3,NegativeWidth=1,PositiveHeight=1.5,NegativeHeight=1.5)
    ActiveRegionSize=0.1
    Debug=(bDrawDebug=false,DebugScale=25,DebugSegments=20,DebugThickness=0.25)
    DriverAxis=(X=-1,Y=0,Z=0)
    DriverItem=(Type=Bone,Name="upperarm_r")
    FalloffRegionScaleFactors=(PositiveWidth=1,NegativeWidth=1,PositiveHeight=1,NegativeHeight=1)
    FalloffSize=0.4
    FlipHeightScaling=true
    FlipWidthScaling=false
    OptionalParentItem=(Type=Bone,Name="spine_05")
    RotationOffset=(X=90,Y=0,Z=0)
   )
   n11 = RigVMVariableNode() {variable=upperarmBack_L_0}
   n12 = RigVMVariableNode() {variable=upperarmBack_L_1}
   n13 = RigVMVariableNode() {variable=upperarmBack_L_2}
   n14 = RigVMVariableNode() {variable=upperarmBack_L_3}
   n15 = RigVMVariableNode() {variable=upperarmBack_L_4}
   n16 = RigVMVariableNode() {variable=upperarmBack_L_5}
   n17 = RigVMVariableNode() {variable=upperarmBack_L_6}
   > &RigUnit_ModifyTransforms(
    ItemToModify="((Item=(Type=Bone,Name=\"deform_upperarmStrap_up_r\"),Transform=(Rotation=(X=0,Y=0,Z=0,W=1),Translation=(X=0.7754,Y=4.1261,Z=-0.3928),Scale3D=(X=1,Y=1,Z=1))),(Item=(Type=Bone,Name=\"deform_strap_front_r_01\"),Transform=(Rotation=(X=0.068,Y=-0.0502,Z=-0.0034,W= ...(truncated)"
    WeightMinimum=0
    Weight=@n10.OutputParam
    ItemToModify.0.Transform=@n11.Value
    ItemToModify.1.Transform=@n12.Value
    ItemToModify.2.Transform=@n13.Value
    ItemToModify.3.Transform=@n14.Value
    ItemToModify.4.Transform=@n15.Value
    ItemToModify.5.Transform=@n16.Value
    ItemToModify.6.Transform=@n17.Value
   )
   > n18 = &RigUnit_SphericalPoseReader(
    ActiveRegionScaleFactors=(PositiveWidth=1,NegativeWidth=1,PositiveHeight=1,NegativeHeight=1)
    ActiveRegionSize=0.1
    Debug=(bDrawDebug=false,bDraw2D=False,bDrawLocalAxes=False,DebugScale=25,DebugSegments=20,DebugThickness=0.25)
    DriverAxis=(X=-1,Y=0,Z=0)
    DriverItem=(Type=Bone,Name="upperarm_r")
    FalloffRegionScaleFactors=(PositiveWidth=0.2,NegativeWidth=1,PositiveHeight=1,NegativeHeight=1)
    FalloffSize=0.3
    FlipHeightScaling=False
    FlipWidthScaling=False
    OptionalParentItem=(Type=Bone,Name="spine_05")
    RotationOffset=(X=180,Y=-118,Z=-8)
   )
   > &RigUnit_ModifyTransforms(
    ItemToModify="((Item=(Type=Bone,Name=\"deform_upperarmStrap_up_r\"),Transform=(Rotation=(X=0,Y=0,Z=0,W=1),Translation=(X=0,Y=0,Z=0),Scale3D=(X=1,Y=1,Z=1))),(Item=(Type=Bone,Name=\"deform_strap_front_r_01\"),Transform=(Rotation=(X=0,Y=0,Z=0,W=1),Translation=(X=0,Y=0,Z=0),Sca ...(truncated)"
    WeightMinimum=0
    Weight=@n18.OutputParam
   )
   > n19 = &RigUnit_SphericalPoseReader(
    ActiveRegionScaleFactors=(PositiveWidth=1,NegativeWidth=1,PositiveHeight=1,NegativeHeight=1)
    ActiveRegionSize=0.1
    Debug=(bDrawDebug=false,bDraw2D=False,bDrawLocalAxes=False,DebugScale=25,DebugSegments=20,DebugThickness=0.25)
    DriverAxis=(X=-1,Y=0,Z=0)
    DriverItem=(Type=Bone,Name="upperarm_r")
    FalloffRegionScaleFactors=(PositiveWidth=1,NegativeWidth=1,PositiveHeight=1,NegativeHeight=1)
    FalloffSize=0.2
    FlipHeightScaling=False
    FlipWidthScaling=False
    OptionalParentItem=(Type=Bone,Name="spine_05")
    RotationOffset=(X=93.3557,Y=6.1885,Z=-130.199)
   )
   > &RigUnit_ModifyTransforms(
    ItemToModify="((Item=(Type=Bone,Name=\"deform_upperarmStrap_up_r\"),Transform=(Rotation=(X=-0.0763,Y=-0.2995,Z=0.077,W=0.9479),Translation=(X=2.1396,Y=-0.8177,Z=3.997),Scale3D=(X=1,Y=1,Z=1))),(Item=(Type=Bone,Name=\"deform_strap_front_r_01\"),Transform=(Rotation=(X=0.1859,Y ...(truncated)"
    WeightMinimum=0
    Weight=@n19.OutputParam
   )
   &RigUnit_ParentConstraint(
    AdvancedSettings=(InterpolationType=Average,RotationOrderForFilter=XZY)
    Child=(Type=Bone,Name="deform_strap_front_l_master")
    Parents=((Item=(Type=Bone,Name="clavicle_l"),Weight=0.8),(Item=(Type=Bone,Name="spine_05"),Weight=0.2))
    Weight=1
    ExecutePin=@n1.B
   )
   > &RigUnit_ParentConstraint(
    AdvancedSettings=(InterpolationType=Average,RotationOrderForFilter=XZY)
    Child=(Type=Bone,Name="deform_strap_back_l_master")
    Parents=((Item=(Type=Bone,Name="clavicle_l"),Weight=0.8),(Item=(Type=Bone,Name="spine_05"),Weight=0.2))
    Weight=1
   )
   > &RigUnit_ParentConstraint(
    AdvancedSettings=(InterpolationType=Average,RotationOrderForFilter=XZY)
    Child=(Type=Bone,Name="deform_arm_fr_base_l")
    Parents=((Item=(Type=Bone,Name="clavicle_l"),Weight=0.8),(Item=(Type=Bone,Name="spine_05"),Weight=0.2))
    Weight=1
   )
   > &RigUnit_ParentConstraint(
    AdvancedSettings=(InterpolationType=Average,RotationOrderForFilter=XZY)
    Child=(Type=Bone,Name="deform_arm_bk_base_l")
    Parents=((Item=(Type=Bone,Name="clavicle_l"),Weight=0.9),(Item=(Type=Bone,Name="spine_05"),Weight=0.1))
    Weight=1
   )
   > &RigUnit_RotationConstraintLocalSpaceOffset(
    AdvancedSettings=(InterpolationType=Average,RotationOrderForFilter=ZYX)
    Child=(Type=Bone,Name="deform_upperarmStrap_base_l")
    Filter=(bX=True,bY=false,bZ=false)
    Parents=((Item=(Type=Bone,Name="upperarm_twist_01_l"),Weight=1))
    Weight=1
   )
   > &RigUnit_ParentConstraint(
    AdvancedSettings=(InterpolationType=Average,RotationOrderForFilter=XZY)
    Child=(Type=Bone,Name="deform_upperarmStrap_up_base_l")
    Parents=((Item=(Type=Bone,Name="clavicle_l"),Weight=0.8),(Item=(Type=Bone,Name="upperarm_l"),Weight=0.2))
    Weight=1
   )
   > n20 = &RigUnit_SphericalPoseReader(
    ActiveRegionScaleFactors=(PositiveWidth=1,NegativeWidth=1,PositiveHeight=1,NegativeHeight=1)
    ActiveRegionSize=0.8
    Debug=(bDrawDebug=false,bDraw2D=False,bDrawLocalAxes=False,DebugScale=25,DebugSegments=20,DebugThickness=0.25)
    DriverAxis=(X=1,Y=0,Z=0)
    DriverItem=(Type=Bone,Name="upperarm_l")
    FalloffRegionScaleFactors=(PositiveWidth=1,NegativeWidth=1,PositiveHeight=1,NegativeHeight=1)
    FalloffSize=0.99
    FlipHeightScaling=False
    FlipWidthScaling=False
    OptionalParentItem=(Type=Bone,Name="clavicle_l")
    RotationOffset=(X=0,Y=90,Z=0)
   )
   > &RigUnit_ModifyTransforms(
    ItemToModify="((Item=(Type=Bone,Name=\"deform_upperarmStrap_up_l\"),Transform=(Rotation=(X=0,Y=0,Z=0,W=1),Translation=(X=0,Y=0,Z=1),Scale3D=(X=1,Y=1,Z=1))),(Item=(Type=Bone,Name=\"deform_strap_front_l_01\"),Transform=(Rotation=(X=0,Y=0,Z=0,W=1),Translation=(X=0,Y=0,Z=0),Sca ...(truncated)"
    WeightMinimum=0
    Weight=@n20.OutputParam
   )
   > n21 = &RigUnit_SphericalPoseReader(
    ActiveRegionScaleFactors=(PositiveWidth=2.5,NegativeWidth=3,PositiveHeight=3,NegativeHeight=3)
    ActiveRegionSize=0.1
    Debug=(bDrawDebug=false,DebugScale=25,DebugSegments=20,DebugThickness=0.25)
    DriverAxis=(X=1,Y=0,Z=0)
    DriverItem=(Type=Bone,Name="upperarm_l")
    FalloffRegionScaleFactors=(PositiveWidth=1,NegativeWidth=1,PositiveHeight=1,NegativeHeight=1)
    FalloffSize=0.5
    OptionalParentItem=(Type=Bone,Name="spine_05")
    RotationOffset=(X=0,Y=0,Z=0)
   )
   > &RigUnit_ModifyTransforms(
    ItemToModify="((Item=(Type=Bone,Name=\"deform_upperarmStrap_up_l\"),Transform=(Rotation=(X=0,Y=-0.3037,Z=0,W=0.9528),Translation=(X=-6.3689,Y=0,Z=1.7675),Scale3D=(X=1,Y=1,Z=1))),(Item=(Type=Bone,Name=\"deform_strap_front_l_01\"),Transform=(Rotation=(X=0,Y=0,Z=0,W=1),Transla ...(truncated)"
    WeightMinimum=0
    Weight=@n21.OutputParam
   )
   > n22 = &RigUnit_SphericalPoseReader(
    ActiveRegionScaleFactors=(PositiveWidth=0.5,NegativeWidth=3,PositiveHeight=1.5,NegativeHeight=1.5)
    ActiveRegionSize=0.1
    Debug=(bDrawDebug=false,DebugScale=25,DebugSegments=20,DebugThickness=0.25)
    DriverAxis=(X=1,Y=0,Z=0)
    DriverItem=(Type=Bone,Name="upperarm_l")
    FalloffRegionScaleFactors=(PositiveWidth=1,NegativeWidth=1,PositiveHeight=1,NegativeHeight=1)
    FalloffSize=0.4
    OptionalParentItem=(Type=Bone,Name="spine_05")
    RotationOffset=(X=90,Y=0,Z=0)
   )
   > Mirror_7_transforms_2_2 = RigVMFunctionReferenceNode(
    transform_1=(Rotation=(X=-0.2075,Y=-0.084,Z=-0.4802,W=0.8481),Translation=(X=1.4019,Y=-2.0829,Z=0.9133),Scale3D=(X=1,Y=1,Z=1))
    transform_5=(Rotation=(X=0,Y=0,Z=0,W=1),Translation=(X=0,Y=0,Z=0),Scale3D=(X=1,Y=1,Z=1))
    transform=@n5.Value
    transform_2=@n6.Value
    transform_3=@n7.Value
    transform_4=@n8.Value
    transform_6=@n9.Value
   )
   > &RigUnit_ModifyTransforms(
    ItemToModify="((Item=(Type=Bone,Name=\"deform_upperarmStrap_up_l\"),Transform=(Rotation=(X=-0.2411,Y=-0.0061,Z=-0.0244,W=0.9702),Translation=(X=-1.9821,Y=1.4306,Z=0.2952),Scale3D=(X=1,Y=1,Z=1))),(Item=(Type=Bone,Name=\"deform_strap_front_l_01\"),Transform=(Rotation=(X=0.184 ...(truncated)"
    WeightMinimum=0
    ItemToModify.0.Transform=@Mirror_7_transforms_2_2.mirrorTransform
    ItemToModify.2.Transform=@Mirror_7_transforms_2_2.mirrorTransform_2
    ItemToModify.3.Transform=@Mirror_7_transforms_2_2.mirrorTransform_3
    ItemToModify.4.Transform=@Mirror_7_transforms_2_2.mirrorTransform_4
    ItemToModify.6.Transform=@Mirror_7_transforms_2_2.mirrorTransform_6
    Weight=@n22.OutputParam
   )
   > n23 = &RigUnit_SphericalPoseReader(
    ActiveRegionScaleFactors=(PositiveWidth=3,NegativeWidth=1,PositiveHeight=1.5,NegativeHeight=1.5)
    ActiveRegionSize=0.1
    Debug=(bDrawDebug=false,DebugScale=25,DebugSegments=20,DebugThickness=0.25)
    DriverAxis=(X=1,Y=0,Z=0)
    DriverItem=(Type=Bone,Name="upperarm_l")
    FalloffRegionScaleFactors=(PositiveWidth=1,NegativeWidth=1,PositiveHeight=1,NegativeHeight=1)
    FalloffSize=0.4
    OptionalParentItem=(Type=Bone,Name="spine_05")
    RotationOffset=(X=90,Y=0,Z=180)
   )
   > Mirror_7_transforms_2_1_1 = RigVMFunctionReferenceNode(
    transform=@n11.Value
    transform_1=@n12.Value
    transform_2=@n13.Value
    transform_3=@n14.Value
    transform_4=@n15.Value
    transform_5=@n16.Value
    transform_6=@n17.Value
   )
   > &RigUnit_ModifyTransforms(
    ItemToModify="((Item=(Type=Bone,Name=\"deform_upperarmStrap_up_l\"),Transform=(Rotation=(X=0,Y=0,Z=0,W=1),Translation=(X=0,Y=0,Z=0),Scale3D=(X=1,Y=1,Z=1))),(Item=(Type=Bone,Name=\"deform_strap_front_l_01\"),Transform=(Rotation=(X=0,Y=0,Z=0,W=1),Translation=(X=0,Y=0,Z=0),Sca ...(truncated)"
    WeightMinimum=0
    ItemToModify.0.Transform=@Mirror_7_transforms_2_1_1.mirrorTransform
    ItemToModify.1.Transform=@Mirror_7_transforms_2_1_1.mirrorTransform_1
    ItemToModify.2.Transform=@Mirror_7_transforms_2_1_1.mirrorTransform_2
    ItemToModify.3.Transform=@Mirror_7_transforms_2_1_1.mirrorTransform_3
    ItemToModify.4.Transform=@Mirror_7_transforms_2_1_1.mirrorTransform_4
    ItemToModify.5.Transform=@Mirror_7_transforms_2_1_1.mirrorTransform_5
    ItemToModify.6.Transform=@Mirror_7_transforms_2_1_1.mirrorTransform_6
    Weight=@n23.OutputParam
   )
   > n24 = &RigUnit_SphericalPoseReader(
    ActiveRegionScaleFactors=(PositiveWidth=1,NegativeWidth=1,PositiveHeight=1,NegativeHeight=1)
    ActiveRegionSize=0.1
    Debug=(bDrawDebug=false,bDraw2D=False,bDrawLocalAxes=False,DebugScale=25,DebugSegments=20,DebugThickness=0.25)
    DriverAxis=(X=1,Y=0,Z=0)
    DriverItem=(Type=Bone,Name="upperarm_l")
    FalloffRegionScaleFactors=(PositiveWidth=0.2,NegativeWidth=1,PositiveHeight=1,NegativeHeight=1)
    FalloffSize=0.2
    FlipHeightScaling=False
    FlipWidthScaling=False
    OptionalParentItem=(Type=Bone,Name="spine_05")
    RotationOffset=(X=180,Y=62,Z=-8)
   )
   > &RigUnit_ModifyTransforms(
    ItemToModify="((Item=(Type=Bone,Name=\"deform_upperarmStrap_up_l\"),Transform=(Rotation=(X=0,Y=0,Z=0,W=1),Translation=(X=0,Y=0,Z=0),Scale3D=(X=1,Y=1,Z=1))),(Item=(Type=Bone,Name=\"deform_strap_front_l_01\"),Transform=(Rotation=(X=0,Y=0,Z=0,W=1),Translation=(X=0,Y=0,Z=0),Sca ...(truncated)"
    WeightMinimum=0
    Weight=@n24.OutputParam
   )
   > n25 = &RigUnit_SphericalPoseReader(
    ActiveRegionScaleFactors=(PositiveWidth=1,NegativeWidth=1,PositiveHeight=1,NegativeHeight=1)
    ActiveRegionSize=0.1
    Debug=(bDrawDebug=false,bDraw2D=False,bDrawLocalAxes=False,DebugScale=25,DebugSegments=20,DebugThickness=0.25)
    DriverAxis=(X=1,Y=0,Z=0)
    DriverItem=(Type=Bone,Name="upperarm_l")
    FalloffRegionScaleFactors=(PositiveWidth=1,NegativeWidth=1,PositiveHeight=1,NegativeHeight=1)
    FalloffSize=0.2
    FlipHeightScaling=False
    FlipWidthScaling=False
    OptionalParentItem=(Type=Bone,Name="spine_05")
    RotationOffset=(X=46.9413,Y=5.6589,Z=4.6641)
   )
   > &RigUnit_ModifyTransforms(
    ItemToModify="((Item=(Type=Bone,Name=\"deform_upperarmStrap_up_l\"),Transform=(Rotation=(X=0.1334,Y=-0.1015,Z=0.0568,W=0.9842),Translation=(X=0.6281,Y=-2.0222,Z=1.6473),Scale3D=(X=1,Y=1,Z=1))),(Item=(Type=Bone,Name=\"deform_strap_front_l_01\"),Transform=(Rotation=(X=0.0097, ...(truncated)"
    WeightMinimum=0
    Weight=@n25.OutputParam
   )
   n26 = &RigUnit_GetRelativeTransformForItem(Child=(Type=Bone,Name="calf_l"), Parent=(Type=Bone,Name="thigh_l"), bChildInitial=False, bParentInitial=False)
   RigVMVariableNode(
    Value=&RigVMFunction_MathDoubleRemap(
     SourceMaximum=130
     SourceMinimum=120
     TargetMaximum=1
     TargetMinimum=0
     Value=&RigVMFunction_MathQuaternionToEuler(RotationOrder=ZYX, Value=@n26.RelativeTransform.Rotation).Result.Z
    ).Result
    ExecuteContext=@RigVMFunction_Sequence_1.B
   ) {variable=L calf bend}
   n27 = &RigVMFunction_MathQuaternionToEuler(
    RotationOrder=ZYX
    Value=&RigUnit_GetRelativeTransformForItem(Child=(Type=Bone,Name="calf_r"), Parent=(Type=Bone,Name="thigh_r"), bChildInitial=False, bParentInitial=False).RelativeTransform.Rotation
   )
   > RigVMVariableNode(Value=&RigVMFunction_MathDoubleRemap(SourceMaximum=130, SourceMinimum=120, TargetMaximum=1, TargetMinimum=0, Value=@n27.Result.Z).Result) {variable=R calf bend}
   > RigVMVariableNode(Value=&RigVMFunction_MathDoubleRemap(SourceMaximum=85, SourceMinimum=68, TargetMaximum=1, TargetMinimum=0, Value=@n27.Result.Z).Result) {variable=R calf bend_walk}
   n28 = &RigUnit_GetRelativeTransformForItem(Child=(Type=Bone,Name="thigh_l"), Parent=(Type=Bone,Name="pelvis"), bChildInitial=False, bParentInitial=False)
   > RigVMVariableNode(Value=&RigVMFunction_MathDoubleRemap(
     SourceMaximum=-105
     SourceMinimum=-90
     TargetMaximum=1
     TargetMinimum=0
     Value=&RigVMFunction_MathQuaternionToEuler(RotationOrder=ZYX, Value=@n28.RelativeTransform.Rotation).Result.Z
    ).Result) {variable=L thigh max rotation}
   n29 = &RigVMFunction_MathQuaternionToEuler(
    RotationOrder=ZYX
    Value=&RigUnit_GetRelativeTransformForItem(Child=(Type=Bone,Name="thigh_r"), Parent=(Type=Bone,Name="pelvis"), bChildInitial=False, bParentInitial=False).RelativeTransform.Rotation
   )
   > RigVMVariableNode(Value=&RigVMFunction_MathDoubleRemap(SourceMaximum=110, SourceMinimum=90, TargetMaximum=1, TargetMinimum=0, Value=@n29.Result.Z).Result) {variable=R thigh max rotation}
   > RigVMVariableNode(Value=&RigVMFunction_MathDoubleRemap(SourceMaximum=145, SourceMinimum=110, TargetMaximum=1, TargetMinimum=0, Value=@n29.Result.Z).Result) {variable=R thigh back}
   > &RigUnit_ModifyTransforms(
    ItemToModify=((Item=(Type=Bone,Name="dyn_skirt_l_02"),Transform=(Rotation=(X=-0.2578,Y=0.0868,Z=0.0076,W=0.9623),Translation=(X=0,Y=0,Z=0),Scale3D=(X=1,Y=1,Z=1))))
    WeightMinimum=0
    Weight=&RigVMFunction_MathDoubleMul(B=RigVMVariableNode() {variable=L calf bend}.Value, A=RigVMVariableNode() {variable=L thigh max rotation}.Value).Result
   )
   > &RigUnit_ModifyTransforms(
    ItemToModify="((Item=(Type=Bone,Name=\"dyn_skirt_r_02\"),Transform=(Rotation=(X=0.1704,Y=0.0872,Z=-0.03,W=0.9811),Translation=(X=-0.6784,Y=0.3418,Z=-1.8488),Scale3D=(X=1,Y=1,Z=1))),(Item=(Type=Bone,Name=\"dyn_skirt_r_01\"),Transform=(Rotation=(X=0,Y=0.0872,Z=0,W=0.9962),Tra ...(truncated)"
    WeightMinimum=0
    Weight=&RigVMFunction_MathDoubleMul(A=RigVMVariableNode() {variable=R thigh max rotation}.Value, B=RigVMVariableNode() {variable=R calf bend}.Value).Result
   )
   > &RigUnit_ModifyTransforms(
    ItemToModify="((Item=(Type=Bone,Name=\"dyn_skirt_r_02\"),Transform=(Rotation=(X=0,Y=0.0872,Z=0,W=0.9962),Translation=(X=0,Y=0,Z=0),Scale3D=(X=1,Y=1,Z=1))),(Item=(Type=Bone,Name=\"dyn_skirt_r_03\"),Transform=(Rotation=(X=0,Y=0,Z=0,W=1),Translation=(X=0,Y=0,Z=0),Scale3D=(X=1, ...(truncated)"
    WeightMinimum=0
    Weight=&RigVMFunction_MathDoubleMul(A=RigVMVariableNode() {variable=R thigh max rotation}.Value, B=RigVMVariableNode() {variable=R calf bend_walk}.Value).Result
   )
   AlphaInterp = &RigVMFunction_AlphaInterp(InterpSpeedDecreasing=1, InterpSpeedIncreasing=1, Scale=1, Value=0, bClampResult=False, bInterpResult=False)
   &RigUnit_ModifyTransforms(
    ExecutePin=()
    ItemToModify="((Item=(Type=Bone,Name=\"deform_upperarmStrap_up_r\"),Transform=(Rotation=(X=0,Y=0,Z=0,W=1),Translation=(X=0,Y=0,Z=0),Scale3D=(X=1,Y=1,Z=1))),(Item=(Type=Bone,Name=\"deform_strap_front_r_01\"),Transform=(Rotation=(X=0,Y=0,Z=0,W=1),Translation=(X=0,Y=0,Z=0),Sca ...(truncated)"
    Weight=1
    WeightMinimum=0
   )
   &RigUnit_ModifyTransforms(
    ExecutePin=()
    ItemToModify="((Item=(Type=Bone,Name=\"deform_upperarmStrap_up_l\"),Transform=(Rotation=(X=0,Y=0,Z=0,W=1),Translation=(X=0,Y=0,Z=0),Scale3D=(X=1,Y=1,Z=1))),(Item=(Type=Bone,Name=\"deform_strap_front_l_01\"),Transform=(Rotation=(X=0,Y=0,Z=0,W=1),Translation=(X=0,Y=0,Z=0),Sca ...(truncated)"
    Weight=1
    WeightMinimum=0
   )
   RigVMVariableNode() {variable=upperarmFront_L_1}
   RigVMVariableNode() {variable=R thigh back}
  ]}
]
hierarchy [
 root {loc=<0 0 0>, rot=<0 0 0 1>}
 pelvis {parent=root, loc=<0 -0.5301 98.0571>, rot=<0.0127 0.707 -0.0127 -0.707>}
 spine_01 {parent=pelvis, loc=<6.1208 -0.9328 0>, rot=<0 0 0.034 -0.9994>}
 spine_02 {parent=spine_01, loc=<6.5996 -0.2095 0>, rot=<0 0 0.0238 -0.9997>}
 spine_03 {parent=spine_02, loc=<6.4914 -0.1237 0>, rot=<0 0 0.0149 -0.9999>}
 spine_04 {parent=spine_03, loc=<6.9598 0.0794 0>, rot=<0 0 -0.0132 -0.9999>}
 spine_05 {parent=spine_04, loc=<6.9987 0 0>, rot=<0 0 -0.0015 -1>}
 attach_cape {parent=spine_05, loc=<9.3216 -8.4501 0>, rot=<0.0616 -0.7044 -0.0616 -0.7044>}
 neck_01 {parent=spine_05, loc=<11.3138 0.1529 0>, rot=<0 0 -0.2003 -0.9797>}
 neck_02 {parent=neck_01, loc=<4.5627 0 0>, rot=<0 0 0.0504 -0.9987>}
 head {parent=neck_02, loc=<5.3914 0 0>, rot=<0 0 0.115 -0.9934>}
 hair_root {parent=head, loc=<0 0 0>, rot=<0 0 0 1>}
 dyn_hair_back_master {parent=hair_root, loc=<1.2986 -9.5945 0.4879>, rot=<-0.0915 0.696 0.701 -0.1254>}
 dyn_hair_back_m_01 {parent=dyn_hair_back_master, loc=<0 0 0>, rot=<0 0 0 1>}
 dyn_hair_back_m_02 {parent=dyn_hair_back_m_01, loc=<7.6761 -0.2132 1.7342>, rot=<-0.0021 0.0792 0.0212 -0.9966>}
 dyn_hair_back_m_03 {parent=dyn_hair_back_m_02, loc=<6.9636 0 0>, rot=<0 0.0723 0 -0.9974>}
 dyn_hair_back_m_04 {parent=dyn_hair_back_m_03, loc=<10.6743 0 0.7365>, rot=<0 0.0781 0 -0.9969>}
 dyn_hair_back_l_master {parent=hair_root, loc=<0.7575 -5.0061 -7.5399>, rot=<0.077 -0.9466 0.2919 -0.1128>}
 dyn_hair_back_l_01 {parent=dyn_hair_back_l_master, loc=<0 0 0>, rot=<0 0 0 1>}
 dyn_hair_back_l_02 {parent=dyn_hair_back_l_01, loc=<7.8747 -0.5741 -0.7175>, rot=<-0.0903 -0.0776 0.2104 -0.9703>}
 dyn_hair_back_l_03 {parent=dyn_hair_back_l_02, loc=<8.6505 -1.0387 0.1378>, rot=<-0.121 -0.153 -0.1395 -0.9708>}
 dyn_hair_back_l_04 {parent=dyn_hair_back_l_03, loc=<4.0987 0 0>, rot=<0.0002 -0.0014 0.0319 -0.9995>}
 dyn_hair_back_r_master {parent=hair_root, loc=<0.6764 -4.4652 7.038>, rot=<0.9466 0.077 -0.1128 -0.2919>}
 dyn_hair_back_r_01 {parent=dyn_hair_back_r_master, loc=<0 0 0>, rot=<0 0 0 1>}
 dyn_hair_back_r_02 {parent=dyn_hair_back_r_01, loc=<-8.8568 2.1048 1.3168>, rot=<-0.0903 -0.0776 0.2104 -0.9703>}
 dyn_hair_back_r_b_03 {parent=dyn_hair_back_r_02, loc=<-5.5804 -1.5742 -1.535>, rot=<-0.1172 0.1108 -0.1355 -0.9775>}
 dyn_hair_back_r_03 {parent=dyn_hair_back_r_02, loc=<-7.1461 0.5044 -0.6301>, rot=<-0.121 -0.153 -0.1395 -0.9708>}
 dyn_hair_back_r_04 {parent=dyn_hair_back_r_03, loc=<-8.117 -0.3469 -0.2429>, rot=<0.0002 -0.0014 0.0319 -0.9995>}
 dyn_hair_r_master {parent=hair_root, loc=<4.2273 1.7822 7.4903>, rot=<0.9937 0.0104 -0.0816 0.0765>}
 dyn_hair_r_a_01 {parent=dyn_hair_r_master, loc=<0 0 0>, rot=<0 0 0 1>}
 dyn_hair_r_b_01 {parent=dyn_hair_r_a_01, loc=<-8.8785 -2.6388 1.4084>, rot=<0.4194 -0.2911 -0.3142 -0.8004>}
 dyn_hair_r_b_02 {parent=dyn_hair_r_b_01, loc=<-2.6379 0 0>, rot=<0 0 0 1>}
 dyn_hair_r_c_01 {parent=dyn_hair_r_a_01, loc=<-12.5158 -0.9435 1.1786>, rot=<0.2208 -0.2585 -0.3794 -0.8605>}
 dyn_hair_r_c_02 {parent=dyn_hair_r_c_01, loc=<-3.2814 0 0>, rot=<0 0 0 1>}
 dyn_hair_r_a_02 {parent=dyn_hair_r_a_01, loc=<-7.6301 6.0662 1.195>, rot=<-0.0606 -0.0686 -0.055 -0.9943>}
 dyn_hair_r_a_03 {parent=dyn_hair_r_a_02, loc=<-4.7557 0.2108 -0.0202>, rot=<0 0 0 1>}
 dyn_hair_r_a_04 {parent=dyn_hair_r_a_03, loc=<-4.6201 -4.866 -0.3624>, rot=<0.199 -0.1984 -0.4634 -0.8404>}
 dyn_hair_r_a_05 {parent=dyn_hair_r_a_04, loc=<-3.9386 -0.0001 0>, rot=<0.0774 -0.1362 -0.0995 -0.9826>}
 dyn_hair_l_master {parent=hair_root, loc=<4.3262 1.8771 -8.0921>, rot=<0.0104 -0.9937 -0.0765 -0.0816>}
 dyn_hair_l_a_01 {parent=dyn_hair_l_master, loc=<0 0 0>, rot=<0 0 0 1>}
 dyn_hair_l_d_01 {parent=dyn_hair_l_a_01, loc=<6.0889 3.5149 -2.7777>, rot=<0.646 -0.3243 -0.243 -0.6469>}
 dyn_hair_l_c_01 {parent=dyn_hair_l_a_01, loc=<9.817 -1.4822 -0.9477>, rot=<0.2353 -0.18 -0.2394 -0.9246>}
 dyn_hair_l_c_02 {parent=dyn_hair_l_c_01, loc=<6.0785 1.5598 1.0512>, rot=<0.0676 -0.1706 -0.3166 -0.9306>}
 dyn_hair_l_b_01 {parent=dyn_hair_l_a_01, loc=<9.4855 2.4484 -2.4262>, rot=<0.3096 -0.2759 -0.299 -0.8594>}
 dyn_hair_l_b_02 {parent=dyn_hair_l_b_01, loc=<2.502 0 0.2387>, rot=<0 -0.2301 0 -0.9732>}
 dyn_hair_l_a_02 {parent=dyn_hair_l_a_01, loc=<6.7394 -6.3141 -1.3864>, rot=<-0.0668 0.0086 0.0046 -0.9977>}
 dyn_hair_l_a_03 {parent=dyn_hair_l_a_02, loc=<5.829 0.5349 -1.3741>, rot=<0.0515 -0.1416 -0.2251 -0.9626>}
 dyn_hair_l_a_04 {parent=dyn_hair_l_a_03, loc=<3.9736 1.7182 1.0877>, rot=<0.149 -0.2208 -0.257 -0.929>}
 dyn_hair_l_a_05 {parent=dyn_hair_l_a_04, loc=<6.194 0.0998 1.6458>, rot=<0 0 0 1>}
 dyn_antena_r_master {parent=head, loc=<9.7499 7.1988 2.4179>, rot=<0.0074 0.1833 -0.048 -0.9819>}
 dyn_antena_r_01 {parent=dyn_antena_r_master, loc=<0 0 0>, rot=<0 0 0 1>}
 dyn_antena_r_02 {parent=dyn_antena_r_01, loc=<5.1967 0.1341 -0.7486>, rot=<-0.0844 0.0069 -0.1365 -0.987>}
 dyn_antena_r_03 {parent=dyn_antena_r_02, loc=<4.0121 -0.0937 -0.0605>, rot=<0 0 -0.0763 -0.9971>}
 dyn_antena_l_master {parent=head, loc=<9.7496 7.1988 -2.4179>, rot=<0.1833 -0.0074 0.9819 -0.048>}
 dyn_antena_l_01 {parent=dyn_antena_l_master, loc=<0 0 0>, rot=<0 0 0 1>}
 dyn_antena_l_02 {parent=dyn_antena_l_01, loc=<-5.1973 -0.134 0.7488>, rot=<-0.0844 0.0069 -0.1365 -0.987>}
 dyn_antena_l_03 {parent=dyn_antena_l_02, loc=<-4.0115 0.0935 0.0603>, rot=<0 0 -0.0763 -0.9971>}
 attach_backpack {parent=spine_05, loc=<0 -8.35 0>, rot=<-0.0093 -0.707 0.0093 -0.707>}
 clavicle_l {parent=spine_05, loc=<5.4831 -0.1053 -3.2946>, rot=<0.0465 -0.7289 -0.009 -0.6829>}
 upperarm_l {parent=clavicle_l, loc=<9.8658 -1.2136 -0.4105>, rot=<0.0164 -0.3245 -0.0269 -0.9454>}
 lowerarm_l {parent=upperarm_l, loc=<26.465 -0.8345 0.259>, rot=<0.0036 0.0161 -0.2298 -0.9731>}
 hand_l {parent=lowerarm_l, loc=<22.7355 -0.3203 -0.6105>, rot=<-0.6784 -0.0223 0.0735 -0.7307>}
 thumb_01_l {parent=hand_l, loc=<1.3118 -0.3191 -2.459>, rot=<0.6546 -0.4042 0.0664 -0.6354>}
 thumb_02_l {parent=thumb_01_l, loc=<5.9336 -0.0448 0.0401>, rot=<-0.0037 0.0213 0.298 -0.9543>}
 thumb_03_l {parent=thumb_02_l, loc=<2.2223 0.3682 -0.0907>, rot=<0.0118 -0.013 0.0066 -0.9998>}
 index_metacarpal_l {parent=hand_l, loc=<1.6535 0.0183 -2.96>, rot=<0.0158 -0.0124 -0.0173 -0.9996>}
 index_01_l {parent=index_metacarpal_l, loc=<8.3008 0.0357 -0.0524>, rot=<-0.0044 0.0485 0.0924 -0.9945>}
 index_02_l {parent=index_01_l, loc=<3.3587 -0.3004 0.0452>, rot=<-0.002 0.0027 0.3225 -0.9466>}
 index_03_l {parent=index_02_l, loc=<2.5578 -0.123 -0.0252>, rot=<-0.002 0.0067 0.2558 -0.9667>}
 middle_metacarpal_l {parent=hand_l, loc=<2.1856 -0.1725 -0.5984>, rot=<0.006 0.0168 -0.0329 -0.9993>}
 middle_01_l {parent=middle_metacarpal_l, loc=<7.3275 -0.2447 0.0129>, rot=<-0.0208 0.0042 0.0858 -0.9961>}
 middle_02_l {parent=middle_01_l, loc=<4.0123 0.1401 0.1175>, rot=<-0.0218 0.0258 0.3512 -0.9357>}
 middle_03_l {parent=middle_02_l, loc=<2.6077 0.0534 0.1373>, rot=<0.0082 0.0188 0.1708 -0.9851>}
 ring_metacarpal_l {parent=hand_l, loc=<2.2074 -0.3625 1.3638>, rot=<-0.0284 0.0722 -0.0458 -0.9959>}
 ring_01_l {parent=ring_metacarpal_l, loc=<7.0295 -0.6197 -0.1446>, rot=<-0.0489 -0.0252 0.1892 -0.9804>}
 ring_02_l {parent=ring_01_l, loc=<3.6174 0.4647 -0.1027>, rot=<0 0 0.2806 -0.9598>}
 ring_03_l {parent=ring_02_l, loc=<2.4152 0.114 0.0199>, rot=<0 0 0.1189 -0.9929>}
 pinky_metacarpal_l {parent=hand_l, loc=<1.7616 -0.9753 3.3814>, rot=<-0.0397 0.0814 -0.069 -0.9935>}
 pinky_01_l {parent=pinky_metacarpal_l, loc=<7.1709 -0.4831 -0.1094>, rot=<-0.0654 -0.0035 0.2242 -0.9723>}
 pinky_02_l {parent=pinky_01_l, loc=<2.8274 0.0893 -0.1043>, rot=<0.0037 0.0011 0.2854 -0.9584>}
 pinky_03_l {parent=pinky_02_l, loc=<2.2046 0.0307 -0.0078>, rot=<-0.0002 0.0098 0.1445 -0.9895>}
 deform_wrist_dn_base_l {parent=hand_l, loc=<0 -1.5326 0>, rot=<0 0 0 -1>}
 deform_wrist_dn_l {parent=deform_wrist_dn_base_l, loc=<0 0 0>, rot=<0 0 0 1>}
 deform_wrist_up_base_l {parent=hand_l, loc=<0 2.3009 0>, rot=<0 0 0 -1>}
 deform_wrist_up_l {parent=deform_wrist_up_base_l, loc=<0 0 0>, rot=<0 0 0 1>}
 weapon_l {parent=hand_l, loc=<-0.1974 -3.4111 0.3811>, rot=<0 0 0 1>}
 hand_attach_l {parent=hand_l, loc=<8.5 -3 0>, rot=<0.7071 0.7071 0 0>}
 lowerarm_twist_02_l {parent=lowerarm_l, loc=<10.6707 0 0>, rot=<0 0 0 1>}
 lowerarm_twist_01_l {parent=lowerarm_l, loc=<18.316 0 0>, rot=<0 0 0 1>}
 deform_elbow_in_base_l {parent=lowerarm_l, loc=<0.1834 3.6062 -0.2572>, rot=<0 0 0 -1>}
 deform_elbow_in_l {parent=deform_elbow_in_base_l, loc=<0 0 0>, rot=<0 0 0 1>}
 deform_elbow_out_base_l {parent=lowerarm_l, loc=<0 -4 0>, rot=<0 0 0 1>}
 deform_elbow_out_l {parent=deform_elbow_out_base_l, loc=<0 0 0>, rot=<0 0 0 1>}
 upperarm_twist_01_l {parent=upperarm_l, loc=<7.2135 -0.1134 0>, rot=<0 0 0 1>}
 deform_upperarm_bk_base_l {parent=upperarm_twist_01_l, loc=<-6.4053 -4.3746 0.314>, rot=<0 0 0 -1>}
 deform_upperarm_bk_l {parent=deform_upperarm_bk_base_l, loc=<0 0 0>, rot=<0 0 0 1>}
 deform_upperarm_fr_base_l {parent=upperarm_twist_01_l, loc=<-5.6312 5.092 0.3354>, rot=<0 0 0 -1>}
 deform_upperarm_fr_l {parent=deform_upperarm_fr_base_l, loc=<0 0 0>, rot=<0 0 0 1>}
 deform_upperarm_up_base_l {parent=upperarm_twist_01_l, loc=<-8.891 0.4825 3.9517>, rot=<0 0 0 -1>}
 deform_upperarm_up_l {parent=deform_upperarm_up_base_l, loc=<0 0 0>, rot=<0 0 0 1>}
 upperarm_twist_02_l {parent=upperarm_l, loc=<19.7998 -0.1134 0>, rot=<0 0 0 1>}
 deform_upperarmStrap_base_l {parent=upperarm_l, loc=<0 0 0>, rot=<0 0 0 1>}
 deform_upperarmStrap_up_base_l {parent=deform_upperarmStrap_base_l, loc=<2.3442 0.056 4.893>, rot=<0 0 0 1>}
 deform_upperarmStrap_up_l {parent=deform_upperarmStrap_up_base_l, loc=<0 0 0>, rot=<0 0 0 1>}
 deform_armStrap_fr_base_l {parent=deform_upperarmStrap_base_l, loc=<2.374 4.6145 0.9236>, rot=<0.9343 -0.0039 0.3563 -0.0065>}
 deform_armStrap_fr_l {parent=deform_armStrap_fr_base_l, loc=<0 0 0>, rot=<0 0 0 1>}
 deform_armStrap_bk_base_l {parent=deform_upperarmStrap_base_l, loc=<3.567 -4.9984 -1.2567>, rot=<0.9343 -0.0039 0.3563 -0.0065>}
 deform_armStrap_bk_l {parent=deform_armStrap_bk_base_l, loc=<0 0 0>, rot=<0 0 0 1>}
 clavicle_r {parent=spine_05, loc=<5.4831 -0.1053 3.2946>, rot=<0.7289 0.0465 -0.6829 0.009>}
 upperarm_r {parent=clavicle_r, loc=<-9.8658 1.2136 0.4105>, rot=<0.0164 -0.3245 -0.0269 -0.9454>}
 lowerarm_r {parent=upperarm_r, loc=<-26.465 0.8345 -0.259>, rot=<0.0036 0.0161 -0.2298 -0.9731>}
 hand_r {parent=lowerarm_r, loc=<-22.7355 0.3203 0.6106>, rot=<-0.6784 -0.0223 0.0735 -0.7307>}
 thumb_01_r {parent=hand_r, loc=<-1.3118 0.3191 2.459>, rot=<0.6546 -0.4042 0.0664 -0.6354>}
 thumb_02_r {parent=thumb_01_r, loc=<-5.9336 0.0448 -0.0401>, rot=<-0.0037 0.0213 0.298 -0.9543>}
 thumb_03_r {parent=thumb_02_r, loc=<-2.2223 -0.3682 0.0907>, rot=<0.0118 -0.013 0.0066 -0.9998>}
 index_metacarpal_r {parent=hand_r, loc=<-1.6535 -0.0183 2.96>, rot=<0.0158 -0.0124 -0.0173 -0.9996>}
 index_01_r {parent=index_metacarpal_r, loc=<-8.3008 -0.0357 0.0524>, rot=<-0.0044 0.0485 0.0924 -0.9945>}
 index_02_r {parent=index_01_r, loc=<-3.3587 0.3004 -0.0452>, rot=<-0.002 0.0027 0.3225 -0.9466>}
 index_03_r {parent=index_02_r, loc=<-2.5578 0.123 0.0252>, rot=<-0.002 0.0067 0.2558 -0.9667>}
 middle_metacarpal_r {parent=hand_r, loc=<-2.1856 0.1725 0.5984>, rot=<0.006 0.0168 -0.0329 -0.9993>}
 middle_01_r {parent=middle_metacarpal_r, loc=<-7.3275 0.2447 -0.0129>, rot=<-0.0208 0.0042 0.0858 -0.9961>}
 middle_02_r {parent=middle_01_r, loc=<-4.0123 -0.1401 -0.1175>, rot=<-0.0218 0.0258 0.3512 -0.9357>}
 middle_03_r {parent=middle_02_r, loc=<-2.6077 -0.0534 -0.1373>, rot=<0.0082 0.0188 0.1708 -0.9851>}
 ring_metacarpal_r {parent=hand_r, loc=<-2.2074 0.3625 -1.3638>, rot=<-0.0284 0.0722 -0.0458 -0.9959>}
 ring_01_r {parent=ring_metacarpal_r, loc=<-7.0295 0.6197 0.1446>, rot=<-0.0489 -0.0252 0.1892 -0.9804>}
 ring_02_r {parent=ring_01_r, loc=<-3.6174 -0.4647 0.1027>, rot=<0 0 0.2806 -0.9598>}
 ring_03_r {parent=ring_02_r, loc=<-2.4152 -0.114 -0.0199>, rot=<0 0 0.1189 -0.9929>}
 pinky_metacarpal_r {parent=hand_r, loc=<-1.7616 0.9753 -3.3814>, rot=<-0.0397 0.0814 -0.069 -0.9935>}
 pinky_01_r {parent=pinky_metacarpal_r, loc=<-7.1709 0.4831 0.1094>, rot=<-0.0654 -0.0035 0.2242 -0.9723>}
 pinky_02_r {parent=pinky_01_r, loc=<-2.8274 -0.0893 0.1043>, rot=<0.0037 0.0011 0.2854 -0.9584>}
 pinky_03_r {parent=pinky_02_r, loc=<-2.2046 -0.0307 0.0078>, rot=<-0.0002 0.0098 0.1445 -0.9895>}
 deform_wrist_dn_base_r {parent=hand_r, loc=<-0.0001 1.5328 0>, rot=<0 0 0 -1>}
 deform_wrist_dn_r {parent=deform_wrist_dn_base_r, loc=<0 0 0>, rot=<0 0 0 1>}
 deform_wrist_up_base_r {parent=hand_r, loc=<-0.0002 -2.3007 -0.0001>, rot=<0 0 0 -1>}
 deform_wrist_up_r {parent=deform_wrist_up_base_r, loc=<0 0 0>, rot=<0 0 0 1>}
 weapon_r {parent=hand_r, loc=<0.1974 3.4111 -0.3811>, rot=<0 0 0 1>}
 hand_attach_r {parent=hand_r, loc=<-8.5001 2.9999 0>, rot=<0 0 -0.7071 -0.7071>}
 lowerarm_twist_02_r {parent=lowerarm_r, loc=<-10.6707 0 0>, rot=<0 0 0 -1>}
 lowerarm_twist_01_r {parent=lowerarm_r, loc=<-18.316 0 0>, rot=<0 0 0 -1>}
 deform_elbow_in_base_r {parent=lowerarm_r, loc=<-0.1834 -3.6062 0.2572>, rot=<0 0 0 -1>}
 deform_elbow_in_r {parent=deform_elbow_in_base_r, loc=<0 0 0>, rot=<0 0 0 1>}
 deform_elbow_out_base_r {parent=lowerarm_r, loc=<0 4 -0.0001>, rot=<0 0 0 -1>}
 deform_elbow_out_r {parent=deform_elbow_out_base_r, loc=<0 0 0>, rot=<0 0 0 1>}
 upperarm_twist_01_r {parent=upperarm_r, loc=<-7.2135 0.1134 0>, rot=<0 0 0 -1>}
 deform_upperarm_bk_base_r {parent=upperarm_twist_01_r, loc=<6.4053 4.3746 -0.314>, rot=<0 0 0 -1>}
 deform_upperarm_bk_r {parent=deform_upperarm_bk_base_r, loc=<0 0 0>, rot=<0 0 0 1>}
 deform_upperarm_fr_base_r {parent=upperarm_twist_01_r, loc=<5.6312 -5.092 -0.3354>, rot=<0 0 0 -1>}
 deform_upperarm_fr_r {parent=deform_upperarm_fr_base_r, loc=<0 0 0>, rot=<0 0 0 1>}
 deform_upperarm_up_base_r {parent=upperarm_twist_01_r, loc=<8.891 -0.4825 -3.9517>, rot=<0 0 0 -1>}
 deform_upperarm_up_r {parent=deform_upperarm_up_base_r, loc=<0 0 0>, rot=<0 0 0 1>}
 upperarm_twist_02_r {parent=upperarm_r, loc=<-19.7998 0.1134 0>, rot=<0 0 0 -1>}
 deform_upperarmStrap_base_r {parent=upperarm_r, loc=<0.0002 0 -0.0003>, rot=<0 0 0 -1>}
 deform_upperarmStrap_up_base_r {parent=deform_upperarmStrap_base_r, loc=<-2.3445 -0.0561 -4.8927>, rot=<0 0 0 -1>}
 deform_upperarmStrap_up_r {parent=deform_upperarmStrap_up_base_r, loc=<0 0 0>, rot=<0 0 0 -1>}
 deform_armStrap_fr_base_r {parent=deform_upperarmStrap_base_r, loc=<-2.3743 -4.6145 -0.9233>, rot=<0.9343 -0.0039 0.3563 -0.0065>}
 deform_armStrap_fr_r {parent=deform_armStrap_fr_base_r, loc=<0 0 0>, rot=<0 0 0 1>}
 deform_armStrap_bk_base_r {parent=deform_upperarmStrap_base_r, loc=<-3.5671 4.9984 1.2569>, rot=<0.9343 -0.0039 0.3563 -0.0065>}
 deform_armStrap_bk_r {parent=deform_armStrap_bk_base_r, loc=<0 0 0>, rot=<0 0 0 1>}
 deform_strap_front_r_master {parent=spine_05, loc=<-2.7438 5.6302 11.271>, rot=<0.6871 -0.0605 0.6982 0.1919>}
 deform_strap_front_r_01 {parent=deform_strap_front_r_master, loc=<0 0 0>, rot=<0 0 0 1>}
 deform_strap_back_r_master {parent=spine_05, loc=<-4.1053 -5.7365 11.4105>, rot=<0.6313 -0.1375 0.753 -0.1247>}
 deform_strap_back_r_01 {parent=deform_strap_back_r_master, loc=<0 0 0>, rot=<0 0 0 1>}
 deform_strap_back_l_master {parent=spine_05, loc=<-4.1055 -5.7365 -11.4105>, rot=<0.1375 0.6313 -0.1247 -0.753>}
 deform_strap_back_l_01 {parent=deform_strap_back_l_master, loc=<0 0 0>, rot=<0 0 0 1>}
 deform_strap_front_l_master {parent=spine_05, loc=<-2.744 5.6302 -11.271>, rot=<0.0605 0.6871 0.1919 -0.6982>}
 deform_strap_front_l_01 {parent=deform_strap_front_l_master, loc=<0 0 0>, rot=<0 0 0 1>}
 deform_arm_fr_base_r {parent=spine_05, loc=<4.0725 3.6604 12.0443>, rot=<0.0284 -0.7065 -0.0284 -0.7065>}
 deform_arm_fr_r {parent=deform_arm_fr_base_r, loc=<0 0 0>, rot=<0 0 0 1>}
 deform_arm_bk_base_r {parent=spine_05, loc=<4.9225 -6.65 13.0638>, rot=<0.0284 -0.7065 -0.0284 -0.7065>}
 deform_arm_bk_r {parent=deform_arm_bk_base_r, loc=<0 0 0>, rot=<0 0 0 1>}
 deform_arm_fr_base_l {parent=spine_05, loc=<4.0728 3.6604 -12.0443>, rot=<0.7065 0.0284 -0.7065 0.0284>}
 deform_arm_fr_l {parent=deform_arm_fr_base_l, loc=<0 0 0>, rot=<0 0 0 1>}
 deform_arm_bk_base_l {parent=spine_05, loc=<4.9227 -6.65 -13.0638>, rot=<0.7065 0.0284 -0.7065 0.0284>}
 deform_arm_bk_l {parent=deform_arm_bk_base_l, loc=<0 0 0>, rot=<0 0 0 1>}
 deform_pec_aim_base_l {parent=spine_04, loc=<3.2456 9.3309 -1.2237>, rot=<0.0603 -0.4033 0.3318 -0.8507>}
 deform_pec_aim_l {parent=deform_pec_aim_base_l, loc=<0 0 0>, rot=<0 0 0 1>}
 deform_pec_base_l {parent=deform_pec_aim_l, loc=<9.3665 0 0>, rot=<0 0 0 1>}
 deform_pec_l {parent=deform_pec_base_l, loc=<0 0 0>, rot=<0 0 0 1>}
 deform_pec_aim_base_r {parent=spine_04, loc=<3.2452 9.3309 1.2237>, rot=<-0.4033 -0.0603 0.8507 0.3318>}
 deform_pec_aim_r {parent=deform_pec_aim_base_r, loc=<0 0 0>, rot=<0 0 0 1>}
 deform_pec_base_r {parent=deform_pec_aim_r, loc=<-9.3667 -0.0003 -0.0004>, rot=<0 0 0 1>}
 deform_pec_r {parent=deform_pec_base_r, loc=<0 0 0>, rot=<0 0 0 1>}
 thigh_l {parent=pelvis, loc=<-1.0855 -0.2109 -8.7053>, rot=<0.0009 0.0355 0.0155 -0.9993>}
 calf_l {parent=thigh_l, loc=<-42.5284 -0.0019 -0.7297>, rot=<0.0006 -0.0161 -0.0391 -0.9991>}
 foot_l {parent=calf_l, loc=<-46.7401 -2.5014 -0.592>, rot=<0.0427 -0.0238 0.0353 -0.9982>}
 ball_l {parent=foot_l, loc=<-3.5658 10.8066 -0.1544>, rot=<0.0001 -0.0001 -0.7063 -0.7079>}
 deform_foot_front_base_l {parent=foot_l, loc=<2.8967 7.2369 1.086>, rot=<0.7082 0.0284 -0.7045 -0.0357>}
 deform_foot_front_l {parent=deform_foot_front_base_l, loc=<0 0 0>, rot=<0 0 0 1>}
 deform_foot_back_base_l {parent=foot_l, loc=<5.6897 -1.9277 0.6579>, rot=<0.7082 0.0284 -0.7045 -0.0357>}
 deform_foot_back_l {parent=deform_foot_back_base_l, loc=<0 0 0>, rot=<0 0 0 1>}
 calf_twist_02_l {parent=calf_l, loc=<-19.1546 -1.7943 -0.2638>, rot=<0.0237 0.0095 -0.0035 -0.9997>}
 calf_twist_01_l {parent=calf_l, loc=<-35.8483 -1.9197 -0.5794>, rot=<0.0237 0.0095 -0.0035 -0.9997>}
 dyn_zipper_master_l {parent=calf_twist_01_l, loc=<0 -1.8197 -4.5201>, rot=<0.0667 0.1983 0.9752 0.072>}
 dyn_zipper_01_l {parent=dyn_zipper_master_l, loc=<0 0 0>, rot=<0 0 0 1>}
 deform_knee_in_base_l {parent=calf_l, loc=<0.8031 -5.0231 0.2456>, rot=<0 0 0 -1>}
 deform_knee_in_l {parent=deform_knee_in_base_l, loc=<0 0 0>, rot=<0 0 0 1>}
 deform_knee_out_base_l {parent=calf_l, loc=<0 4 0>, rot=<0 0 0 -1>}
 deform_knee_out_l {parent=deform_knee_out_base_l, loc=<0 0 0>, rot=<0 0 0 1>}
 thigh_twist_01_l {parent=thigh_l, loc=<-27.952 0 0>, rot=<0.0013 0 0 -1>}
 deform_groin_base_l {parent=thigh_l, loc=<0.4801 5.6776 -0.2486>, rot=<0 0 0 -1>}
 deform_groin_l {parent=deform_groin_base_l, loc=<0 0 0>, rot=<0 0 0 1>}
 dyn_skirt_l_master {parent=thigh_l, loc=<-7.9327 -2.1006 -7.9733>, rot=<0.1796 -0.9772 0.1114 -0.0191>}
 dyn_skirt_l_01 {parent=dyn_skirt_l_master, loc=<0 0 0>, rot=<0 0 0 1>}
 dyn_skirt_l_02 {parent=dyn_skirt_l_01, loc=<9.369 0 0>, rot=<-0.0745 -0.1343 0.0366 -0.9875>}
 dyn_skirt_l_03 {parent=dyn_skirt_l_02, loc=<11.6413 0 0>, rot=<-0.0887 -0.036 0.0404 -0.9946>}
 thigh_r {parent=pelvis, loc=<-1.0853 -0.2109 8.7053>, rot=<0.0355 -0.0009 0.9993 0.0155>}
 calf_r {parent=thigh_r, loc=<42.5286 0.0019 0.7297>, rot=<0.0006 -0.0161 -0.0391 -0.9991>}
 foot_r {parent=calf_r, loc=<46.7401 2.5014 0.592>, rot=<0.0427 -0.0238 0.0353 -0.9982>}
 ball_r {parent=foot_r, loc=<3.5658 -10.8066 0.1544>, rot=<0.0001 -0.0001 -0.7063 -0.7079>}
 deform_foot_back_base_r {parent=foot_r, loc=<-5.6897 1.9277 -0.6579>, rot=<0.7082 0.0284 -0.7045 -0.0357>}
 deform_foot_back_r {parent=deform_foot_back_base_r, loc=<0 0 0>, rot=<0 0 0 1>}
 deform_foot_front_base_r {parent=foot_r, loc=<-2.8967 -7.2369 -1.086>, rot=<0.7082 0.0284 -0.7045 -0.0357>}
 deform_foot_front_r {parent=deform_foot_front_base_r, loc=<0 0 0>, rot=<0 0 0 1>}
 calf_twist_02_r {parent=calf_r, loc=<19.1546 1.7943 0.2638>, rot=<0.0237 0.0095 -0.0035 -0.9997>}
 calf_twist_01_r {parent=calf_r, loc=<35.8483 1.9197 0.5794>, rot=<0.0237 0.0095 -0.0035 -0.9997>}
 dyn_zipper_master_r {parent=calf_twist_01_r, loc=<0 1.8197 4.5201>, rot=<0.0667 0.1983 0.9752 0.072>}
 dyn_zipper_01_r {parent=dyn_zipper_master_r, loc=<0 0 0>, rot=<0 0 0 1>}
 deform_knee_in_base_r {parent=calf_r, loc=<-0.8032 5.0231 -0.2457>, rot=<0 0 0 -1>}
 deform_knee_in_r {parent=deform_knee_in_base_r, loc=<0 0 0>, rot=<0 0 0 1>}
 deform_knee_out_base_r {parent=calf_r, loc=<-0.0001 -4 0>, rot=<0 0 0 -1>}
 deform_knee_out_r {parent=deform_knee_out_base_r, loc=<0 0 0>, rot=<0 0 0 1>}
 thigh_twist_01_r {parent=thigh_r, loc=<27.952 0 0>, rot=<0.0013 0 0 -1>}
 deform_groin_base_r {parent=thigh_r, loc=<-0.4801 -5.6776 0.2486>, rot=<0 0 0 -1>}
 deform_groin_r {parent=deform_groin_base_r, loc=<0 0 0>, rot=<0 0 0 1>}
 dyn_skirt_r_master {parent=thigh_r, loc=<7.9327 2.1006 7.9733>, rot=<0.9772 0.1796 0.0191 0.1114>}
 dyn_skirt_r_01 {parent=dyn_skirt_r_master, loc=<0 0 0>, rot=<0 0 0 1>}
 dyn_skirt_r_02 {parent=dyn_skirt_r_01, loc=<9.369 0 0>, rot=<0.0243 0.1385 0.0178 -0.9899>}
 dyn_skirt_r_03 {parent=dyn_skirt_r_02, loc=<11.638 -0.2728 -0.0631>, rot=<0 0 0 1>}
 deform_glute_base_l {parent=pelvis, loc=<-7.3466 -9.0213 -9.1245>, rot=<0.0172 0.031 -0.4477 -0.8935>}
 deform_glute_l {parent=deform_glute_base_l, loc=<0 0 0>, rot=<0 0 0 1>}
 deform_glute_base_r {parent=pelvis, loc=<-7.3464 -9.0213 9.1245>, rot=<0.031 -0.0172 0.8935 -0.4477>}
 deform_glute_r {parent=deform_glute_base_r, loc=<0 0 0>, rot=<0 0 0 1>}
 dyn_skirt_m_master {parent=pelvis, loc=<-0.0691 -10.952 0>, rot=<0.0636 -0.7042 0.7042 -0.0636>}
 dyn_skirt_m_01 {parent=dyn_skirt_m_master, loc=<0 0 0>, rot=<0 0 0 1>}
 dyn_skirt_m_02 {parent=dyn_skirt_m_01, loc=<10.2219 0 0>, rot=<0 -0.0571 0 -0.9984>}
 dyn_skirt_m_03 {parent=dyn_skirt_m_02, loc=<13.3788 0 0>, rot=<0 -0.0359 0 -0.9994>}
 dyn_skirt_m_04 {parent=dyn_skirt_m_03, loc=<13.8063 0 0>, rot=<0 0 0 1>}
 dyn_skirt_m_02_L {parent=dyn_skirt_m_02, loc=<0 7 0>, rot=<0 0 0 1>}
 dyn_skirt_m_02_R {parent=dyn_skirt_m_02, loc=<0 -7 0>, rot=<0 0 0 1>}
 deform_belt_back_base_r {parent=pelvis, loc=<6.7458 -7.1927 6.7694>, rot=<-0.0127 -0.707 0.0127 -0.707>}
 deform_belt_back_r {parent=deform_belt_back_base_r, loc=<0 0 0>, rot=<0 0 0 1>}
 deform_belt_front_base_r {parent=pelvis, loc=<3.7322 7.1232 6.795>, rot=<-0.0127 -0.707 0.0127 -0.707>}
 deform_belt_front_r {parent=deform_belt_front_base_r, loc=<0 0 0>, rot=<0 0 0 1>}
 deform_belt_back_base_l {parent=pelvis, loc=<6.7455 -7.1927 -6.7694>, rot=<0.707 -0.0127 -0.707 -0.0127>}
 deform_belt_back_l {parent=deform_belt_back_base_l, loc=<0 0 0>, rot=<0 0 0 1>}
 deform_belt_front_base_l {parent=pelvis, loc=<3.7324 7.1232 -6.795>, rot=<0.707 -0.0127 -0.707 -0.0127>}
 deform_belt_front_l {parent=deform_belt_front_base_l, loc=<0 0 0>, rot=<0 0 0 1>}
 root {parent=root, type=Socket, loc=<0 0 0>, rot=<0 0 0 1>}
 attach {parent=root, loc=<0 0 0>, rot=<0 0 0 1>}
 ik_hand_root {parent=root, loc=<0 0 0>, rot=<0 0 0 1>}
 ik_hand_gun {parent=ik_hand_root, loc=<-48.1743 3.3545 104.4381>, rot=<0.6276 0.1309 0.4183 -0.6434>}
 ik_hand_r {parent=ik_hand_gun, loc=<0 0 0>, rot=<0 0 0 1>}
 ik_hand_l {parent=ik_hand_gun, loc=<59.3271 67.6997 34.3524>, rot=<-0.6158 -0.7027 -0.3565 0>}
 ik_foot_root {parent=root, loc=<0 0 0>, rot=<0 0 0 1>}
 ik_foot_l {parent=ik_foot_root, loc=<14.8262 -7.3545 8.296>, rot=<0.0357 0.7045 0.0284 -0.7082>}
 ik_foot_r {parent=ik_foot_root, loc=<-14.8262 -7.3545 8.296>, rot=<0.7082 0.0284 -0.7045 0.0357>}
 attach_fp {parent=root, loc=<0 0 0>, rot=<0 0 0 1>}
 camera_root {parent=root, loc=<0 0 0>, rot=<0 0 0 1>}
 camera {parent=camera_root, loc=<0 0 0>, rot=<0 0 0 1>}
 attach_emote_01 {parent=root, loc=<0 0 0>, rot=<0 0 0 1>}
 attach_emote_02 {parent=root, loc=<0 0 0>, rot=<0 0 0 1>}
 attach_emote_03 {parent=root, loc=<0 0 0>, rot=<0 0 0 1>}
 warp_point_01 {parent=root, loc=<0 0 0>, rot=<0 0 0 1>}
 warp_point_02 {parent=root, loc=<0 0 0>, rot=<0 0 0 1>}
 warp_point_03 {parent=root, loc=<0 0 0>, rot=<0 0 0 1>}
]
variables [
 L calf bend {type=double, default=0}
 L thigh max rotation {type=double, default=0}
 R calf bend {type=double, default=0}
 R calf bend_walk {type=double, default=0}
 R thigh back {type=double, default=0}
 R thigh max rotation {type=double, default=0}
 upperarmBack_L_0 {type=FTransform, default=(Rotation=(X=0,Y=0,Z=0,W=1),Translation=(X=0.7754,Y=4.1261,Z=-0.3928),Scale3D=(X=1,Y=1,Z=1))}
 upperarmBack_L_1 {type=FTransform, default=(Rotation=(X=0.068,Y=-0.0502,Z=-0.0034,W=0.9964),Translation=(X=0,Y=0,Z=0),Scale3D=(X=1,Y=1,Z=1))}
 upperarmBack_L_2 {type=FTransform, default=(Rotation=(X=-0.5049,Y=-0.2858,Z=-0.0087,W=0.8144),Translation=(X=0.5995,Y=2.1803,Z=2.0252),Scale3D=(X=1,Y=1,Z=1))}
 upperarmBack_L_3 {type=FTransform, default=(Rotation=(X=0,Y=0.0924,Z=0,W=0.9957),Translation=(X=1.3514,Y=-1.1474,Z=0),Scale3D=(X=1,Y=1,Z=1))}
 upperarmBack_L_4 {type=FTransform, default=(Rotation=(X=0,Y=0,Z=0,W=1),Translation=(X=18.1385,Y=-8.0547,Z=0),Scale3D=(X=1,Y=1,Z=1))}
 upperarmBack_L_5 {type=FTransform, default=(Rotation=(X=0,Y=0,Z=0,W=1),Translation=(X=6.3235,Y=3.9591,Z=-0.3935),Scale3D=(X=1,Y=1,Z=1))}
 upperarmBack_L_6 {type=FTransform, default=(Rotation=(X=0,Y=0,Z=0,W=1),Translation=(X=2.0624,Y=-13.1679,Z=-13.595),Scale3D=(X=1,Y=1,Z=1))}
 upperarmFront_L_0 {type=FTransform, default=(Rotation=(X=0.2411,Y=0,Z=0,W=0.9705),Translation=(X=3.4484,Y=-1.0999,Z=0.1149),Scale3D=(X=1,Y=1,Z=1))}
 upperarmFront_L_1 {type=FTransform, default=(Rotation=(X=0.3614,Y=-0.4966,Z=-0.2386,W=0.7522),Translation=(X=0.313,Y=-3.006,Z=2.4064),Scale3D=(X=1,Y=1,Z=1))}
 upperarmFront_L_2 {type=FTransform, default=(Rotation=(X=-0.0862,Y=0.0223,Z=-0.0006,W=0.996),Translation=(X=0,Y=0,Z=0),Scale3D=(X=1,Y=1,Z=1))}
 upperarmFront_L_3 {type=FTransform, default=(Rotation=(X=0,Y=0,Z=-0.4298,W=0.9029),Translation=(X=19.3625,Y=1.1835,Z=0),Scale3D=(X=1,Y=1,Z=1))}
 upperarmFront_L_4 {type=FTransform, default=(Rotation=(X=0,Y=0,Z=0,W=1),Translation=(X=0,Y=0,Z=0),Scale3D=(X=1,Y=1,Z=1))}
 upperarmFront_L_6 {type=FTransform, default=(Rotation=(X=0,Y=0,Z=0,W=1),Translation=(X=-4,Y=-3,Z=3),Scale3D=(X=1,Y=1,Z=1))}
]
