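# Physics asset for the SameRabbit medium female player body, built around the dynamic skirt bones.
# Capsule bodies cover pelvis, thighs and calves, with two per thigh, plus the skirt bones.
# The middle skirt bones form a chain from dyn_skirt_m_master through 04 with tight angular limits.
# Side bodies m_02_L and m_02_R lock rotationally to m_02, widening that segment, while dyn_skirt_l_03 rotates freely against thigh_l.
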PhysicsAsset /BRCosmetics/Characters/Player/Female/Medium/Bodies/F_MED_SameRabbit/Meshes/F_MED_SameRabbit_Physics
bodies {
 calf_l [{
   shapeType=Capsule
   length=28.7384
   radius=7.2616
   shapeName=calf_r_capsule
   center=<-23.3566 -3.5856 1.8445>
   rotation=<rot 87.362 1.3394 -180>
  }]
 calf_r [{shapeType=Capsule, length=42, radius=6.7431, shapeName=calf_r_capsule, center=<21.1475 3.5425 -1.7556>, rotation=<rot 90 90 90.0002>}]
 dyn_skirt_l_03 [{
   shapeType=Capsule
   length=9.065
   radius=5.1727
   shapeName=dyn_skirt_l_03_capsule
   center=<3.9004 0.069 -0.1017>
   rotation=<rot 0 85.6735 -88.6654>
  }]
 dyn_skirt_m_01 [{shapeType=Capsule, length=7.4276, radius=3.6738, shapeName=dyn_skirt_m_01_capsule, center=<0.0197 0 1.2>, rotation=<rot -90 -180 0>}]
 dyn_skirt_m_02 [{shapeType=Capsule, length=11.7797, radius=4.0369, shapeName=dyn_skirt_m_02_capsule, rotation=<rot -90 153.4349 116.5651>}]
 dyn_skirt_m_02_L [{shapeType=Capsule, length=10.9041, radius=3.0955, shapeName=dyn_skirt_m_02_L_capsule, center=<-0.0382 1.8962 1>, rotation=<rot 88.9946 0 0>}]
 dyn_skirt_m_02_R [{shapeType=Capsule, length=9.3742, radius=3.4444, shapeName=dyn_skirt_m_02_R_capsule, center=<0 -2.3353 1>, rotation=<rot 87.652 180 180>}]
 dyn_skirt_m_03 [{shapeType=Capsule, length=5.1867, radius=9.059, shapeName=dyn_skirt_m_03_capsule, center=<1.8502 0 -0.5787>, rotation=<rot 90 90 -90>}]
 dyn_skirt_m_04 [{shapeType=Capsule, length=6.3482, radius=5.7187, shapeName=dyn_skirt_m_04_capsule, center=<6.1423 0 -0.0763>, rotation=<rot -90 180 180>}]
 dyn_skirt_m_master [{
   shapeType=Capsule
   length=15
   radius=3.2088
   shapeName=dyn_skirt_m_master_capsule
   center=<-6.6837 -0.0038 -2.4013>
   rotation=<rot -86.7959 -180 90.0335>
  }]
 pelvis [{shapeType=Capsule, length=10.9406, radius=9.8639, center=<-1.1961 -0.3389 0.1658>, rotation=<rot 4.2639 -92.3093 -0.7304>}]
 thigh_l [
  {shapeType=Capsule, length=12.5, radius=9.5, shapeName=thigh_l_capsule, center=<-18.4106 0.0001 0.5627>, rotation=<rot -90 0 -0.0002>}
  {shapeType=Capsule, length=10, radius=8.5, shapeName=thigh_l_capsule2, center=<-35.1991 -0.9382 0>, rotation=<rot -90 180 179.9998>}
 ]
 thigh_r [
  {shapeType=Capsule, length=15, radius=9.5, shapeName=thigh_l_capsule, center=<16.6785 0.9901 2.2527>, rotation=<rot -90 0 179.9998>}
  {shapeType=Capsule, length=13, radius=7.5, shapeName=thigh_r_capsule, center=<35.044 -0.0001 0>, rotation=<rot -90 180 179.9999>}
 ]
}
constraints [
 {bone1Name=dyn_skirt_m_01, bone2Name=dyn_skirt_m_master, swing1LimitDegrees=8, swing2LimitDegrees=20, twistLimitDegrees=0}
 {bone1Name=dyn_skirt_m_02, bone2Name=dyn_skirt_m_01, swing1LimitDegrees=15, swing2LimitDegrees=20, twistLimitDegrees=0}
 {bone1Name=dyn_skirt_m_03, bone2Name=dyn_skirt_m_02, swing1LimitDegrees=15, swing2LimitDegrees=30, twistLimitDegrees=0}
 {bone1Name=dyn_skirt_m_04, bone2Name=dyn_skirt_m_03, swing1LimitDegrees=15, swing2LimitDegrees=30, twistLimitDegrees=8}
 {
  bone1Name=dyn_skirt_m_02_L
  bone2Name=dyn_skirt_m_02
  swing1LimitDegrees=45
  swing1Motion=Locked
  swing2LimitDegrees=45
  swing2Motion=Locked
  twistLimitDegrees=45
  twistMotion=Locked
 }
 {
  bone1Name=dyn_skirt_m_02_R
  bone2Name=dyn_skirt_m_02
  swing1LimitDegrees=45
  swing1Motion=Locked
  swing2LimitDegrees=45
  swing2Motion=Locked
  twistLimitDegrees=45
  twistMotion=Locked
 }
 {
  bone1Name=dyn_skirt_l_03
  bone2Name=thigh_l
  swing1LimitDegrees=45
  swing1Motion=Free
  swing2LimitDegrees=45
  swing2Motion=Free
  twistLimitDegrees=45
  twistMotion=Free
 }
]
