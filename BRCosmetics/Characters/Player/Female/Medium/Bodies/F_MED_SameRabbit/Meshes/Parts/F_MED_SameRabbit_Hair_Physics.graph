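# Physics asset for the hair part of the female medium SameRabbit player body, defining collision and joint limits for left and right hair chains.
# Each chain hangs from a master bone locked to the head, runs five segments and forks at the third, with a shorter right branch.
# Swing is tight at the roots and looser further down, with twist locked throughout.
# Head, neck, upper spine, clavicle and upper-arm capsules provide body collision shapes.
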PhysicsAsset /BRCosmetics/Characters/Player/Female/Medium/Bodies/F_MED_SameRabbit/Meshes/Parts/F_MED_SameRabbit_Hair_Physics
bodies {
 clavicle_l [{
   shapeType=Capsule
   length=7
   radius=7.8623
   shapeName=clavicle_l_capsule
   center=<7.8309 -2.3453 -1.5384>
   rotation=<rot -87.1875 179.9933 -179.9935>
  }]
 clavicle_r [{
   shapeType=Capsule
   length=3.6244
   radius=7.7564
   shapeName=clavicle_r_capsule
   center=<-5.3041 2.2593 1.1701>
   rotation=<rot -80 179.9996 -179.9998>
  }]
 dyn_hair_l_a_01 [{shapeType=Capsule, length=3, radius=2, shapeName=dyn_hair_l_a_01_capsule, center=<1.4327 0 0>, rotation=<rot -90 0 0.0002>}]
 dyn_hair_l_a_02 [{shapeType=Capsule, length=3, radius=2, shapeName=dyn_hair_l_a_02_capsule, center=<1.4327 0 0>, rotation=<rot -90 90 -89.9997>}]
 dyn_hair_l_a_03 [{shapeType=Capsule, length=4, radius=3, shapeName=dyn_hair_l_a_03_capsule, center=<1.4327 -0.964 0>, rotation=<rot -90 -179.8812 179.8816>}]
 dyn_hair_l_a_04 [{shapeType=Capsule, length=3, radius=2, shapeName=dyn_hair_l_a_04_capsule, center=<1.4327 0 0>, rotation=<rot -90 -179.9999 -179.9997>}]
 dyn_hair_l_a_05 [{shapeType=Capsule, length=3, radius=2, shapeName=dyn_hair_l_a_05_capsule, center=<1.4327 0 0>, rotation=<rot -90 0 0.0004>}]
 dyn_hair_l_b_04 [{shapeType=Capsule, length=3, radius=2, shapeName=dyn_hair_l_b_04_capsule, center=<1.4327 0 0>, rotation=<rot -90 -90 90.0004>}]
 dyn_hair_l_b_05 [{shapeType=Capsule, length=3, radius=2, shapeName=dyn_hair_l_b_05_capsule, center=<1.4327 0 0>, rotation=<rot -90 -179.9999 -179.9997>}]
 dyn_hair_l_master [{shapeType=Capsule, length=0.5, radius=1, shapeName=dyn_hair_l_master_capsule}]
 dyn_hair_r_a_01 [{shapeType=Capsule, length=3, radius=2, shapeName=dyn_hair_r_a_01_capsule, center=<1.4327 0 0>, rotation=<rot -90 0.0002 -0.0005>}]
 dyn_hair_r_a_02 [{shapeType=Capsule, length=3, radius=2, shapeName=dyn_hair_r_a_02_capsule, center=<1.4327 0 0>, rotation=<rot -90 -179.9999 179.9996>}]
 dyn_hair_r_a_03 [{
   shapeType=Capsule
   length=3
   radius=2.5
   shapeName=dyn_hair_r_a_03_capsule
   center=<1.4327 2.0816 -0.4802>
   rotation=<rot -90 179.8944 -179.8948>
  }]
 dyn_hair_r_a_04 [{shapeType=Capsule, length=3.5, radius=1.5, shapeName=dyn_hair_r_a_04_capsule, center=<1.4327 0 -0.3418>, rotation=<rot -90 -180 180>}]
 dyn_hair_r_a_05 [{shapeType=Capsule, length=2.5, radius=1.5, shapeName=dyn_hair_r_a_05_capsule, center=<1.4327 0 0>, rotation=<rot -90 179.8417 -179.8421>}]
 dyn_hair_r_b_04 [{
   shapeType=Capsule
   length=3
   radius=2
   shapeName=dyn_hair_r_b_04_capsule
   center=<2.2233 0.0715 -0.5705>
   rotation=<rot -50 179.9994 -169.9999>
  }]
 dyn_hair_r_master [{shapeType=Capsule, length=0.5, radius=1, shapeName=dyn_hair_r_master_capsule}]
 head [
  {shapeType=Capsule, length=7.4107, radius=6.4163, shapeName=head_capsule, center=<1.339 3.3931 0>, rotation=<rot -90 180 176.4588>}
  {shapeType=Capsule, length=4.5767, radius=6.0468, shapeName=head_capsule, center=<3.1579 0.0894 0>, rotation=<rot -90 172.875 98.7767>}
 ]
 neck_02 [{
   shapeType=Capsule
   length=6.2564
   radius=6.3036
   shapeName=neck_02_capsule
   center=<-5.0057 -0.2024 0>
   rotation=<rot 90 -179.9963 -179.9962>
  }]
 spine_04 [
  {shapeType=Capsule, length=8.3819, radius=10.0825, shapeName=spine_04_capsule, center=<2.7775 1.6505 0>}
  {shapeType=Capsule, length=7, radius=8, shapeName=spine_04_capsule, center=<0 5.3511 0>}
 ]
 spine_05 [
  {shapeType=Capsule, length=27.0967, radius=7.2436, shapeName=spine_05_capsule, center=<4.2264 -0.3272 0>}
  {shapeType=Capsule, length=13.2665, radius=6.9208, shapeName=spine_05_capsule, center=<0.3105 1.1259 0>}
  {shapeType=Capsule, length=6, radius=8, shapeName=spine_05_capsule, center=<18.3775 -1.3007 0>, rotation=<rot -90 90 -80.0001>}
 ]
 upperarm_l [{shapeType=Capsule, length=27.5163, radius=5.4837, shapeName=upperarm_l_capsule, center=<13.8301 -0.6224 -0.5316>, rotation=<rot -90 0 0>}]
 upperarm_r [{
   shapeType=Capsule
   length=28.9431
   radius=5.0569
   shapeName=upperarm_r_capsule
   center=<-17.4205 0.8928 -0.6165>
   rotation=<rot -90 180 177.7086>
  }]
}
constraints [
 {bone1Name=dyn_hair_l_master, bone2Name=head, swing1LimitDegrees=0, swing2LimitDegrees=0, twistLimitDegrees=0}
 {bone1Name=dyn_hair_l_a_01, bone2Name=dyn_hair_l_master, swing1LimitDegrees=20, swing2LimitDegrees=2, twistLimitDegrees=0}
 {bone1Name=dyn_hair_l_a_02, bone2Name=dyn_hair_l_a_01, swing1LimitDegrees=35, swing2LimitDegrees=5, twistLimitDegrees=0}
 {bone1Name=dyn_hair_l_a_03, bone2Name=dyn_hair_l_a_02, swing1LimitDegrees=45, swing2LimitDegrees=15, twistLimitDegrees=0}
 {bone1Name=dyn_hair_l_a_04, bone2Name=dyn_hair_l_a_03, swing1LimitDegrees=30, swing2LimitDegrees=20, twistLimitDegrees=0}
 {bone1Name=dyn_hair_l_a_05, bone2Name=dyn_hair_l_a_04, swing1LimitDegrees=30, swing2LimitDegrees=20, twistLimitDegrees=0}
 {bone1Name=dyn_hair_l_b_04, bone2Name=dyn_hair_l_a_03, swing1LimitDegrees=30, swing2LimitDegrees=20, twistLimitDegrees=0}
 {bone1Name=dyn_hair_l_b_05, bone2Name=dyn_hair_l_b_04, swing1LimitDegrees=30, swing2LimitDegrees=20, twistLimitDegrees=0}
 {bone1Name=dyn_hair_r_a_02, bone2Name=dyn_hair_r_a_01, swing1LimitDegrees=35, swing2LimitDegrees=5, twistLimitDegrees=0}
 {bone1Name=dyn_hair_r_a_03, bone2Name=dyn_hair_r_a_02, swing1LimitDegrees=30, swing2LimitDegrees=15, twistLimitDegrees=0}
 {bone1Name=dyn_hair_r_a_04, bone2Name=dyn_hair_r_a_03, swing1LimitDegrees=30, swing2LimitDegrees=20, twistLimitDegrees=0}
 {bone1Name=dyn_hair_r_a_05, bone2Name=dyn_hair_r_a_04, swing1LimitDegrees=30, swing2LimitDegrees=20, twistLimitDegrees=0}
 {bone1Name=dyn_hair_r_b_04, bone2Name=dyn_hair_r_a_03, swing1LimitDegrees=30, swing2LimitDegrees=20, twistLimitDegrees=0}
 {bone1Name=dyn_hair_r_a_01, bone2Name=dyn_hair_r_master, swing1LimitDegrees=20, swing2LimitDegrees=2, twistLimitDegrees=0}
 {bone1Name=dyn_hair_r_master, bone2Name=head, swing1LimitDegrees=0, swing2LimitDegrees=0, twistLimitDegrees=0}
]
