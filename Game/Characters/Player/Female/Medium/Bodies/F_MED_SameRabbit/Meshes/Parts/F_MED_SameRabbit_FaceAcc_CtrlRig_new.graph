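# Control Rig for the SameRabbit female medium player's face accessory part, applying corrective transforms to dynamic hair bones based on body pose.
# Spherical pose readers on head and pelvis store weights in variables; an upperarm_r reader's output is unused.
# A fixed hair offset precedes head-up correctives, with crouch-only layers weighted by multiplied pose values, one also gated by targeting.
# Skydive diving applies its own offset; a final targeting look-up-left corrective is left unconnected to execution.
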ControlRigBlueprint /Game/Characters/Player/Female/Medium/Bodies/F_MED_SameRabbit/Meshes/Parts/F_MED_SameRabbit_FaceAcc_CtrlRig_new
graphs [
 RigVMFunctionLibrary {links [], nodes []}
 RigVMFunction_Sequence_1_ContainedGraph {nodes [
   n1 = RigVMFunctionEntryNode()
   RigVMFunction_Sequence = &RigVMFunction_Sequence(ExecuteContext=@n1.ExecuteContext)
   n2 = &RigVMFunction_Sequence(ExecuteContext=@RigVMFunction_Sequence.B)
   RigVMFunctionReturnNode(A=@RigVMFunction_Sequence.A, B=@n2.A, C=@n2.B)
  ]}
 RigVMModel {nodes [
   RigUnit_BeginExecution = &RigUnit_BeginExecution()
   RigVMFunction_Sequence = RigVMAggregateNode(ExecuteContext=@RigUnit_BeginExecution.ExecutePin)
   n1 = &RigUnit_SphericalPoseReader(
    ActiveRegionScaleFactors=(PositiveWidth=2,NegativeWidth=0.6,PositiveHeight=2,NegativeHeight=2)
    ActiveRegionSize=0.1
    Debug=(bDrawDebug=false,bDraw2D=False,bDrawLocalAxes=False,DebugScale=25,DebugSegments=20,DebugThickness=0.25)
    DriverAxis=(X=1,Y=0,Z=0)
    DriverItem=(Type=Bone,Name="head")
    FalloffRegionScaleFactors=(PositiveWidth=0.7,NegativeWidth=0.5,PositiveHeight=0.7,NegativeHeight=0.7)
    FalloffSize=0.3
    FlipHeightScaling=False
    FlipWidthScaling=False
    OptionalParentItem=(Type=Bone,Name="None")
    RotationOffset=(X=82.1852,Y=16.0762,Z=-120.4624)
    ExecutePin=@RigVMFunction_Sequence.A
   )
   > RigVMVariableNode(Value=@n1.OutputParam) {variable="Head Up "}
   > n2 = &RigUnit_SphericalPoseReader(
    ActiveRegionScaleFactors=(PositiveWidth=0.3,NegativeWidth=0.7,PositiveHeight=1,NegativeHeight=1)
    ActiveRegionSize=0.1
    Debug=(bDrawDebug=false,bDraw2D=False,bDrawLocalAxes=False,DebugScale=25,DebugSegments=20,DebugThickness=0.25)
    DriverAxis=(X=0,Y=1,Z=0)
    DriverItem=(Type=Bone,Name="pelvis")
    FalloffRegionScaleFactors=(PositiveWidth=0.2,NegativeWidth=0.4,PositiveHeight=0.4,NegativeHeight=0.4)
    FalloffSize=0.22
    FlipHeightScaling=False
    FlipWidthScaling=False
    OptionalParentItem=(Type=Bone,Name="None")
    RotationOffset=(X=70,Y=0,Z=18.2708)
   )
   > RigVMVariableNode(Value=@n2.OutputParam) {variable=Pelvis Low}
   > n3 = &RigUnit_SphericalPoseReader(
    ActiveRegionScaleFactors=(PositiveWidth=1,NegativeWidth=1,PositiveHeight=0.5,NegativeHeight=1)
    ActiveRegionSize=0.1
    Debug=(bDrawDebug=false,bDraw2D=False,bDrawLocalAxes=False,DebugScale=25,DebugSegments=20,DebugThickness=0.25)
    DriverAxis=(X=0,Y=1,Z=0)
    DriverItem=(Type=Bone,Name="head")
    FalloffRegionScaleFactors=(PositiveWidth=1,NegativeWidth=1,PositiveHeight=1,NegativeHeight=1)
    FalloffSize=0.22
    FlipHeightScaling=False
    FlipWidthScaling=False
    OptionalParentItem=(Type=Bone,Name="spine_05")
    RotationOffset=(X=136.2209,Y=72.5786,Z=-19.9062)
   )
   > RigVMVariableNode(Value=@n3.OutputParam) {variable=Look Up Left}
   > &RigUnit_SphericalPoseReader(
    ActiveRegionScaleFactors=(PositiveWidth=2.2,NegativeWidth=2.2,PositiveHeight=2.2,NegativeHeight=2.2)
    ActiveRegionSize=0.11
    Debug=(bDrawDebug=false,bDraw2D=false,bDrawLocalAxes=true,DebugScale=25,DebugSegments=20,DebugThickness=0.25)
    DriverAxis=(X=-1,Y=0,Z=0)
    DriverItem=(Type=Bone,Name="upperarm_r")
    FalloffRegionScaleFactors=(PositiveWidth=1,NegativeWidth=1,PositiveHeight=4,NegativeHeight=1)
    FalloffSize=0.3
    FlipHeightScaling=false
    FlipWidthScaling=true
    OptionalParentItem=(Type=Bone,Name="spine_02")
    RotationOffset=(X=-140,Y=50,Z=0)
   )
   > n4 = &RigUnit_SphericalPoseReader(
    ActiveRegionScaleFactors=(PositiveWidth=1,NegativeWidth=1,PositiveHeight=1,NegativeHeight=1)
    ActiveRegionSize=0.1
    Debug=(bDrawDebug=false,bDraw2D=False,bDrawLocalAxes=False,DebugScale=25,DebugSegments=20,DebugThickness=0.25)
    DriverAxis=(X=1,Y=0,Z=0)
    DriverItem=(Type=Bone,Name="head")
    FalloffRegionScaleFactors=(PositiveWidth=0.7,NegativeWidth=0.7,PositiveHeight=0.7,NegativeHeight=0.7)
    FalloffSize=0.22
    FlipHeightScaling=False
    FlipWidthScaling=False
    OptionalParentItem=(Type=Bone,Name="None")
    RotationOffset=(X=94.7838,Y=-26.3245,Z=-142.8814)
   )
   > RigVMVariableNode(Value=@n4.OutputParam) {variable=Head Back Drastic}
   &RigUnit_ModifyTransforms(
    ItemToModify="((Item=(Type=Bone,Name=\"dyn_hair_r_master\"),Transform=(Rotation=(X=0,Y=0,Z=0,W=1),Translation=(X=0,Y=0,Z=1),Scale3D=(X=1,Y=1,Z=1.5))),(Item=(Type=Bone,Name=\"dyn_hair_back_r_master\"),Transform=(Rotation=(X=0,Y=0,Z=0,W=1),Translation=(X=0,Y=0.5089,Z=-1),Scal ...(truncated)"
    Weight=1
    WeightMinimum=0
    ExecutePin=@RigVMFunction_Sequence.B
   )
   AlphaInterp = &RigVMFunction_AlphaInterp(
    ClampMax=0.8
    InterpSpeedDecreasing=10
    InterpSpeedIncreasing=10
    Scale=1
    bClampResult=False
    bInterpResult=False
    Value=RigVMVariableNode() {variable="Head Up "}.Value
   )
   > n5 = &RigUnit_ModifyTransforms(
    ItemToModify="((Item=(Type=Bone,Name=\"dyn_hair_back_master\"),Transform=(Rotation=(X=0,Y=-0.2588,Z=0,W=0.9659),Translation=(X=-0.1235,Y=0,Z=2.5654),Scale3D=(X=1,Y=1,Z=1))),(Item=(Type=Bone,Name=\"dyn_hair_back_r_master\"),Transform=(Rotation=(X=-0.2192,Y=-0.36,Z=0.1022,W=0 ...(truncated)"
    WeightMinimum=0
    Weight=@AlphaInterp.Result
   )
   RigVMFunction_ControlFlowBranch = &RigVMFunction_ControlFlowBranch(ExecuteContext=@n5.ExecutePin, Condition=RigVMVariableNode() {variable=is crouching}.Value)
   n6 = RigVMVariableNode() {variable=Head Back Drastic}
   n7 = RigVMVariableNode() {variable=Pelvis Low}
   n8 = RigVMVariableNode() {variable=is targeting}
   n9 = &RigVMFunction_AlphaInterp(
    InterpSpeedDecreasing=25
    InterpSpeedIncreasing=25
    Scale=1
    bClampResult=False
    bInterpResult=False
    Value=&RigVMFunction_MathDoubleMul(
     A=&RigVMFunction_MathDoubleMul(A=@n6.Value, B=@n7.Value).Result
     B=&RigVMFunction_MathBoolToFloat(Value=@n8.Value).Result
    ).Result
   )
   &RigUnit_ModifyTransforms(
    ItemToModify="((Item=(Type=Bone,Name=\"dyn_hair_back_master\"),Transform=(Rotation=(X=0,Y=0,Z=0,W=1),Translation=(X=0,Y=0,Z=0),Scale3D=(X=1,Y=1,Z=1))),(Item=(Type=Bone,Name=\"dyn_hair_back_r_master\"),Transform=(Rotation=(X=-0.0151,Y=-0.3407,Z=0.0858,W=0.9361),Translation=( ...(truncated)"
    WeightMinimum=0
    Weight=@n9.Result
    ExecutePin=@RigVMFunction_ControlFlowBranch.True
   )
   Multiply = &RigVMFunction_MathDoubleMul(A=RigVMVariableNode() {variable="Head Up "}.Value, B=RigVMVariableNode() {variable=Look Up Left}.Value)
   n10 = &RigVMFunction_AlphaInterp(ClampMax=0.8, InterpSpeedDecreasing=15, InterpSpeedIncreasing=15, Scale=1, bClampResult=False, bInterpResult=False, Value=@Multiply.Result)
   > &RigUnit_ModifyTransforms(
    ItemToModify="((Item=(Type=Bone,Name=\"dyn_hair_back_master\"),Transform=(Rotation=(X=0,Y=0,Z=0,W=1),Translation=(X=0,Y=0,Z=0),Scale3D=(X=1,Y=1,Z=1))),(Item=(Type=Bone,Name=\"dyn_hair_back_r_master\"),Transform=(Rotation=(X=0.0788,Y=-0.0616,Z=0.3445,W=0.9334),Translation=(X ...(truncated)"
    WeightMinimum=0
    Weight=@n10.Result
   )
   n11 = &RigVMFunction_ControlFlowBranch(ExecuteContext=@RigVMFunction_Sequence.C, Condition=RigVMVariableNode() {variable=is skydive diving}.Value)
   &RigUnit_ModifyTransforms(
    ItemToModify="((Item=(Type=Bone,Name=\"dyn_hair_back_r_02\"),Transform=(Rotation=(X=0,Y=0,Z=0,W=1),Translation=(X=0,Y=0,Z=1.6874),Scale3D=(X=1,Y=1,Z=1))),(Item=(Type=Bone,Name=\"dyn_hair_back_r_master\"),Transform=(Rotation=(X=0,Y=-0.1736,Z=0,W=0.9848),Translation=(X=-0.256 ...(truncated)"
    Weight=1
    WeightMinimum=0
    ExecutePin=@n11.True
   )
   RigVMFunction_MathBoolToFloat = &RigVMFunction_MathBoolToFloat(Value=RigVMVariableNode() {variable=is targeting}.Value)
   n12 = &RigVMFunction_AlphaInterp(
    InterpSpeedDecreasing=10
    InterpSpeedIncreasing=10
    Scale=1
    bClampResult=False
    bInterpResult=False
    Value=&RigVMFunction_MathDoubleMul(B=@RigVMFunction_MathBoolToFloat.Result, A=RigVMVariableNode() {variable=Look Up Left}.Value).Result
   )
   &RigUnit_ModifyTransforms(
    ExecutePin=()
    ItemToModify="((Item=(Type=Bone,Name=\"dyn_hair_back_master\"),Transform=(Rotation=(X=0,Y=0,Z=0,W=1),Translation=(X=0,Y=0,Z=0),Scale3D=(X=1,Y=1,Z=1))),(Item=(Type=Bone,Name=\"dyn_hair_back_r_master\"),Transform=(Rotation=(X=0,Y=0,Z=0,W=1),Translation=(X=0,Y=0,Z=0),Scale3D=( ...(truncated)"
    WeightMinimum=0
    Weight=@n12.Result
   )
  ]}
]
hierarchy [
 root {loc=<0 0 0>, rot=<0 0 0 1>}
 pelvis {parent=root, loc=<0 -0.5301 98.0571>, rot=<0.0127 0.707 -0.0127 -0.707>}
 spine_01 {parent=pelvis, loc=<6.1208 -0.9328 0>, rot=<0 0 0.034 -0.9994>}
 spine_02 {parent=spine_01, loc=<6.5996 -0.2095 0>, rot=<0 0 0.0238 -0.9997>}
 spine_03 {parent=spine_02, loc=<6.4914 -0.1237 0>, rot=<0 0 0.0149 -0.9999>}
 spine_04 {parent=spine_03, loc=<6.9598 0.0794 0>, rot=<0 0 -0.0132 -0.9999>}
 spine_05 {parent=spine_04, loc=<6.9987 0 0>, rot=<0 0 -0.0015 -1>}
 neck_01 {parent=spine_05, loc=<11.3138 0.1529 0>, rot=<0 0 -0.2003 -0.9797>}
 neck_02 {parent=neck_01, loc=<4.5627 0 0>, rot=<0 0 0.0504 -0.9987>}
 head {parent=neck_02, loc=<5.3914 0 0>, rot=<0 0 0.115 -0.9934>}
 hair_root {parent=head, loc=<1.0719 -2.2846 0>, rot=<0 0 0 1>}
 dyn_hair_r_master {parent=hair_root, loc=<1.8599 1.6836 7.7752>, rot=<0.0112 0.0747 0.9968 -0.0255>}
 dyn_hair_r_a_01 {parent=dyn_hair_r_master, loc=<0 0 0>, rot=<0 0 0 1>}
 dyn_hair_r_a_02 {parent=dyn_hair_r_a_01, loc=<4.1943 0.0465 0.4229>, rot=<0.0594 0.0466 0.0191 -0.997>}
 dyn_hair_r_a_03 {parent=dyn_hair_r_a_02, loc=<5.4826 0.9688 -0.3104>, rot=<-0.0423 -0.1339 0.5726 -0.8077>}
 dyn_hair_r_a_04 {parent=dyn_hair_r_a_03, loc=<4.4785 1.6908 0.1377>, rot=<-0.7758 -0.2567 0.0275 -0.5758>}
 dyn_hair_r_a_05 {parent=dyn_hair_r_a_04, loc=<3.9386 0 0>, rot=<0.7186 0.0562 0.1288 -0.6811>}
 dyn_hair_r_b_04 {parent=dyn_hair_r_a_03, loc=<4.3889 -3.5632 -0.5363>, rot=<0 0 0 1>}
 dyn_hair_back_r_master {parent=hair_root, loc=<-0.3955 -2.1807 7.038>, rot=<0.0039 -0.299 0.9448 -0.1341>}
 dyn_hair_back_r_01 {parent=dyn_hair_back_r_master, loc=<0 0 0>, rot=<0 0 0 -1>}
 dyn_hair_back_r_02 {parent=dyn_hair_back_r_01, loc=<9.1804 0.5185 0.2411>, rot=<0.2801 0.0048 -0.0288 -0.9595>}
 dyn_hair_back_r_b_03 {parent=dyn_hair_back_r_02, loc=<5.7447 -1.1063 1.3227>, rot=<0.149 0.0987 0.0961 -0.9792>}
 dyn_hair_back_r_03 {parent=dyn_hair_back_r_02, loc=<6.7373 1.7724 1.1976>, rot=<0.1449 -0.1333 0.0916 -0.9761>}
 dyn_hair_back_r_04 {parent=dyn_hair_back_r_03, loc=<5.9869 0.1392 0.1687>, rot=<-0.0019 0.0094 -0.0305 -0.9995>}
 dyn_hair_back_l_master {parent=hair_root, loc=<-0.3144 -2.7216 -7.5399>, rot=<0.077 -0.9466 0.2919 -0.1128>}
 dyn_hair_back_l_01 {parent=dyn_hair_back_l_master, loc=<0 0 0>, rot=<0 0 0 1>}
 dyn_hair_back_l_02 {parent=dyn_hair_back_l_01, loc=<7.8747 -0.5741 -0.7175>, rot=<-0.3779 -0.0107 0.224 -0.8983>}
 dyn_hair_back_l_03 {parent=dyn_hair_back_l_02, loc=<7.1843 -0.2515 0.9175>, rot=<-0.0207 -0.1501 0.0307 -0.988>}
 dyn_hair_back_l_04 {parent=dyn_hair_back_l_03, loc=<5.8327 -0.773 -0.2261>, rot=<0.043 -0.0114 -0.0538 -0.9976>}
 dyn_hair_back_master {parent=hair_root, loc=<0.2267 -7.3099 0.4879>, rot=<0.1254 -0.701 0.696 -0.0915>}
 dyn_hair_back_m_01 {parent=dyn_hair_back_master, loc=<0 0 0>, rot=<0 0 0 1>}
 dyn_hair_back_m_02 {parent=dyn_hair_back_m_01, loc=<7.6761 0.2132 -1.7342>, rot=<-0.0021 -0.0792 -0.0212 -0.9966>}
 dyn_hair_back_m_03 {parent=dyn_hair_back_m_02, loc=<5.9595 -0.002 -0.576>, rot=<0 -0.0723 0 -0.9974>}
 dyn_hair_back_m_04 {parent=dyn_hair_back_m_03, loc=<8.8914 -0.0143 -0.1585>, rot=<0 -0.0781 0 -0.9969>}
 dyn_hair_l_master {parent=hair_root, loc=<1.8599 1.6836 -7.7752>, rot=<0.0362 -0.9931 -0.0554 -0.0972>}
 dyn_hair_l_a_01 {parent=dyn_hair_l_master, loc=<0 0 0>, rot=<0 0 0 1>}
 dyn_hair_l_a_02 {parent=dyn_hair_l_a_01, loc=<4.2045 0.1029 -0.2914>, rot=<-0.0391 -0.0828 -0.034 -0.9952>}
 dyn_hair_l_a_03 {parent=dyn_hair_l_a_02, loc=<5.4925 -0.9601 0.1688>, rot=<0.0834 -0.0853 -0.4204 -0.8995>}
 dyn_hair_l_a_04 {parent=dyn_hair_l_a_03, loc=<4.778 -0.0596 0.1353>, rot=<0.0746 -0.1731 0.0114 -0.982>}
 dyn_hair_l_a_05 {parent=dyn_hair_l_a_04, loc=<3.9385 0.0003 -0.0001>, rot=<0.0774 -0.1362 -0.0995 -0.9826>}
 dyn_hair_l_b_04 {parent=dyn_hair_l_a_03, loc=<2.1633 4.6836 0.1063>, rot=<0.0731 -0.2714 0.0189 -0.9595>}
 dyn_hair_l_b_05 {parent=dyn_hair_l_b_04, loc=<3.0723 -0.3821 0.7194>, rot=<0.0774 -0.1362 -0.0995 -0.9826>}
 dyn_hair_back_up_master {parent=hair_root, loc=<4.9556 -6.9881 0.4879>, rot=<0.1254 -0.701 0.696 -0.0915>}
 dyn_hair_back_up_01 {parent=dyn_hair_back_up_master, loc=<0 0 0>, rot=<0 0 0 1>}
 dyn_hair_back_l_up_master {parent=hair_root, loc=<4.5282 -2.1185 -7.6313>, rot=<0.0029 -0.7071 -0.0029 -0.7071>}
 dyn_hair_back_l_up_01 {parent=dyn_hair_back_l_up_master, loc=<0 0 0>, rot=<0 0 0 1>}
 dyn_hair_back_r_up_master {parent=hair_root, loc=<5.0311 -1.3986 7.6848>, rot=<0.7071 0.0029 0.7071 -0.0029>}
 dyn_hair_back_r_up_01 {parent=dyn_hair_back_r_up_master, loc=<0 0 0>, rot=<0 0 0 1>}
 dyn_antena_r_master {parent=head, loc=<9.7499 7.1988 2.4179>, rot=<-0.6947 0.0353 -0.1019 -0.7111>}
 dyn_antena_r_01 {parent=dyn_antena_r_master, loc=<0 0 0>, rot=<0 0 0 1>}
 dyn_antena_r_02 {parent=dyn_antena_r_01, loc=<5.2478 0.1652 -0.1341>, rot=<-0.0722 -0.1434 -0.0929 -0.9826>}
 dyn_antena_r_03 {parent=dyn_antena_r_02, loc=<4.0121 -0.0605 0.0937>, rot=<0 -0.0763 0 -0.9971>}
 dyn_antena_l_master {parent=head, loc=<9.7496 7.1988 -2.4179>, rot=<-0.7111 -0.1019 0.0353 -0.6947>}
 dyn_antena_l_01 {parent=dyn_antena_l_master, loc=<0 0 0>, rot=<0 0 0 1>}
 dyn_antena_l_02 {parent=dyn_antena_l_01, loc=<5.2483 -0.1651 -0.134>, rot=<0.0722 -0.1434 0.0929 -0.9826>}
 dyn_antena_l_03 {parent=dyn_antena_l_02, loc=<4.0115 0.0603 0.0935>, rot=<0 -0.0763 0 -0.9971>}
 clavicle_l {parent=spine_05, loc=<5.4831 -0.1053 -3.2946>, rot=<0.0465 -0.7289 -0.009 -0.6829>}
 upperarm_l {parent=clavicle_l, loc=<9.8658 -1.2136 -0.4105>, rot=<0.0164 -0.3245 -0.0269 -0.9454>}
 clavicle_r {parent=spine_05, loc=<5.4831 -0.1053 3.2946>, rot=<0.7289 0.0465 -0.6829 0.009>}
 upperarm_r {parent=clavicle_r, loc=<-9.8658 1.2136 0.4104>, rot=<0.0164 -0.3245 -0.0269 -0.9454>}
 attach_cape {parent=spine_05, loc=<9.3216 -8.4501 0>, rot=<0.0616 -0.7044 -0.0616 -0.7044>}
 attach_backpack {parent=spine_05, loc=<0 -8.35 0>, rot=<-0.0093 -0.707 0.0093 -0.707>}
]
variables [
 Head Back Drastic {type=double, default=0}
 "Head Up " {type=double, default=0}
 Look Up Left {type=double, default=0}
 Pelvis Low {type=double, default=0}
 is crouching {type=bool, default=False}
 is skydive diving {type=bool, default=False}
 is targeting {type=bool, default=False}
]
