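# Animation Blueprint for the F_MED_SameRabbit face-accessory mesh, adding secondary motion to hair and antennae.
# Copied parent and body-mesh poses, layered at the clavicles, feed two bool-blended branches: backpack RigidBody with skydive offsets, or two RigidBody passes, constraints and AnimDynamics.
# AnimDynamics then swings both antennae on the blended pose.
# The EventGraph updates gravity, backbling and front-end/in-game alpha variables; a Control Rig node sits unconnected.
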AnimBlueprint /BRCosmetics/Characters/Player/Female/Medium/Bodies/F_MED_SameRabbit/Meshes/Parts/F_MED_SameRabbit_FaceAcc_AnimBP
target_skeleton=/Game/Characters/Player/Female/Medium/Bodies/F_MED_SameRabbit/Meshes/Parts/F_MED_SameRabbit_FaceAcc_Skeleton
graphs [
 AnimGraph {nodes [
   n1 = "Constraint - dyn_hair_back_r_02 (PPP)"!(ComponentPose, Alpha=1, ConstraintWeights_0=0.3, ConstraintWeights_1:float=0.7, ConstraintWeights_2:float=0) -> (Pose) {
    alphaBoolBlend {blendOption=Linear}
    boneToModify {boneName=dyn_hair_back_r_02}
    constraintSetup [{targetBone {boneName=spine_05}}, {targetBone {boneName=clavicle_r}}, {targetBone {boneName=dyn_hair_r_a_02}}]
    constraintWeights [1,1,1]
   }
   "Constraint - dyn_hair_back_l_02 (PPP)"!(ComponentPose=@n1, Alpha=1, ConstraintWeights_0=0.4, ConstraintWeights_1:float=0.6, ConstraintWeights_2:float=0) -> (Pose) {
    alphaBoolBlend {blendOption=Linear}
    boneToModify {boneName=dyn_hair_back_l_02}
    constraintSetup [{targetBone {boneName=spine_05}}, {targetBone {boneName=clavicle_l}}, {targetBone {boneName=dyn_hair_l_a_02}}]
    constraintWeights [1,1,1]
   }
   Control Rig!(Source, Alpha, is crouching:bool=Get bIsCrouching(), is targeting:bool=Get bIsTargeting(), is skydive diving:bool=Get bIsSkydiveFloating()) -> (Pose) {
    alphaBoolBlend {blendOption=Linear}
    controlRigAssetReference {blueprintRigClass=/Game/Characters/Player/Female/Medium/Bodies/F_MED_SameRabbit/Meshes/Parts/F_MED_SameRabbit_FaceAcc_CtrlRig_new.F_MED_SameRabbit_FaceAcc_CtrlRig_new_C}
   }
   n2 = RigidBody!(ComponentPose=Local To Component(LocalPose=Use cached pose 'InitialPose'()), Alpha, OverrideWorldGravity:Vector="0,0,-980", ExternalForce) -> (Pose) {
    bOverrideWorldGravity=true
    overridePhysicsAsset=&/F_MED_SameRabbit_Hair_Physics_backpack
    simulationSpace=BaseBoneSpace
    alphaBoolBlend {blendOption=Linear}
    baseBoneRef {boneName=neck_02}
    componentLinearAccScale=<0.4 0.4 0.4>
    componentLinearVelScale=<0.4 0.4 0.4>
    overrideWorldGravity {z=-980}
    simSpaceSettings {maxAngularAcceleration=400, maxAngularVelocity=400, maxLinearAcceleration=400, maxLinearVelocity=400}
   }
   n3 = "Constraint - dyn_hair_back_m_02 (P)"(ComponentPose=@n2, Alpha=1, ConstraintWeights_0=1) {
    alphaBoolBlend {blendOption=Linear}
    boneToModify {boneName=dyn_hair_back_m_02}
    constraintSetup [{targetBone {boneName=spine_05}}]
    constraintWeights [1]
   }
   n4 = "Constraint - dyn_hair_back_l_02 (P)"(ComponentPose=@n3, Alpha=1, ConstraintWeights_0=1) {
    alphaBoolBlend {blendOption=Linear}
    boneToModify {boneName=dyn_hair_back_l_02}
    constraintSetup [{targetBone {boneName=spine_05}}]
    constraintWeights [1]
   }
   n5 = "Constraint - dyn_hair_back_r_02 (P)"(ComponentPose=@n4, Alpha=1, ConstraintWeights_0=1) {
    alphaBoolBlend {blendOption=Linear}
    boneToModify {boneName=dyn_hair_back_r_02}
    constraintSetup [{targetBone {boneName=spine_05}}]
    constraintWeights [1]
   }
   n6 = "Transform (Modify) Bone - Bone: dyn_hair_back_r_02"!(ComponentPose=@n5, bAlphaBoolEnabled:bool=True, Translation="0,0,1", Rotation="0,0,0", Scale="1,1,1") -> (Pose) {
    alphaInputType=Bool
    translationMode=BMM_Additive
    translationSpace=BCS_BoneSpace
    alphaBoolBlend {blendInTime=0.3, blendOption=Linear, blendOutTime=0.3}
    boneToModify {boneName=dyn_hair_back_r_02}
    translation {z=1}
   }
   n7 = Get bIsSkydiveFloating()
   n8 = "Transform (Modify) Bone - Bone: dyn_hair_back_l_02"!(ComponentPose=@n6, bAlphaBoolEnabled:bool=@n7, Translation="5.926,0.000,3.511", Rotation="-23.796,0.000,0.000", Scale="1,1,1") -> (Pose) {
    alphaInputType=Bool
    rotationMode=BMM_Additive
    rotationSpace=BCS_BoneSpace
    translationMode=BMM_Additive
    translationSpace=BCS_BoneSpace
    alphaBoolBlend {blendInTime=0.3, blendOption=Linear, blendOutTime=0.3}
    boneToModify {boneName=dyn_hair_back_l_02}
    rotation=<rot -23.7961 0 0>
    translation=<5.926 0 3.5113>
   }
   n9 = "Transform (Modify) Bone - Bone: dyn_hair_back_r_02"(ComponentPose=@n8, bAlphaBoolEnabled=@n7, Translation="3.389,0.000,2.075", Rotation="0,0,0", Scale="1,1,1") {
    alphaInputType=Bool
    translationMode=BMM_Additive
    translationSpace=BCS_BoneSpace
    alphaBoolBlend {blendInTime=0.3, blendOption=Linear, blendOutTime=0.3}
    boneToModify {boneName=dyn_hair_back_r_02}
    translation {x=3.389, z=2.075}
   }
   n10 = "Transform (Modify) Bone - Bone: dyn_hair_back_m_02"!(ComponentPose=@n9, bAlphaBoolEnabled:bool=@n7, Translation="5.994,-0.000,0.000", Rotation="0,0,0", Scale="1,1,1") -> (Pose) {
    alphaInputType=Bool
    translationMode=BMM_Additive
    translationSpace=BCS_BoneSpace
    alphaBoolBlend {blendInTime=0.3, blendOption=Linear, blendOutTime=0.3}
    boneToModify {boneName=dyn_hair_back_m_02}
    translation {x=5.994}
   }
   n11 = Component To Local(ComponentPose=@n10)
   n12 = RigidBody(ComponentPose=Local To Component(LocalPose=Use cached pose 'InitialPose'()), Alpha=0.7, OverrideWorldGravity="0,0,-980") {
    bOverrideWorldGravity=true
    overridePhysicsAsset=&/F_MED_SameRabbit_Hair_Physics
    simulationSpace=BaseBoneSpace
    alphaBoolBlend {blendOption=Linear}
    baseBoneRef {boneName=neck_02}
    componentLinearAccScale=<0.4 0.4 0.4>
    componentLinearVelScale=<0.4 0.4 0.4>
    overrideWorldGravity {z=-980}
    simSpaceSettings {maxAngularAcceleration=400, maxAngularVelocity=400, maxLinearAcceleration=400, maxLinearVelocity=400}
   }
   n13 = RigidBody(ComponentPose=@n12, Alpha=0.7, OverrideWorldGravity="0,0,-980") {
    bOverrideWorldGravity=true
    overridePhysicsAsset=&/F_MED_SameRabbit_Hair_Physics_FrontEnd
    simulationSpace=BaseBoneSpace
    alphaBoolBlend {blendOption=Linear}
    baseBoneRef {boneName=neck_02}
    componentLinearAccScale=<0.4 0.4 0.4>
    componentLinearVelScale=<0.4 0.4 0.4>
    overrideWorldGravity {z=-980}
    simSpaceSettings {maxAngularAcceleration=400, maxAngularVelocity=400, maxLinearAcceleration=400, maxLinearVelocity=400}
   }
   n14 = "Transform (Modify) Bone - Bone: dyn_hair_l_a_01"(ComponentPose=@n13, Alpha=1, Translation="0.5,-0.5,0", Rotation="0,0,0", Scale="1,1,1") {
    rotationMode=BMM_Additive
    translationMode=BMM_Additive
    translationSpace=BCS_BoneSpace
    alphaBoolBlend {blendOption=Linear}
    boneToModify {boneName=dyn_hair_l_a_01}
    translation=<0.5 -0.5>
   }
   n15 = "Transform (Modify) Bone - Bone: dyn_hair_r_a_01"(ComponentPose=@n14, Alpha=1, Translation="0.029,0.173,1.147", Rotation="0,0,0", Scale="1,1,1") {
    rotationMode=BMM_Additive
    translationMode=BMM_Additive
    translationSpace=BCS_BoneSpace
    alphaBoolBlend {blendOption=Linear}
    boneToModify {boneName=dyn_hair_r_a_01}
    translation=<0.029 0.173 1.147>
   }
   n16 = "Constraint - dyn_hair_back_l_01 (PPP)"!(ComponentPose=@n15, Alpha=1, ConstraintWeights_0=0, ConstraintWeights_1:float=0, ConstraintWeights_2:float=1) -> (Pose) {
    alphaBoolBlend {blendOption=Linear}
    boneToModify {boneName=dyn_hair_back_l_01}
    constraintSetup [{targetBone {boneName=neck_01}}, {targetBone {boneName=clavicle_l}}, {targetBone {boneName=dyn_hair_l_a_01}}]
    constraintWeights [1,1,1]
   }
   n17 = "Constraint - dyn_hair_back_r_01 (PPP)"!(ComponentPose=@n16, Alpha=1, ConstraintWeights_0=0, ConstraintWeights_1:float=0, ConstraintWeights_2:float=1) -> (Pose) {
    alphaBoolBlend {blendOption=Linear}
    boneToModify {boneName=dyn_hair_back_r_01}
    constraintSetup [{targetBone {boneName=neck_01}}, {targetBone {boneName=clavicle_r}}, {targetBone {boneName=dyn_hair_r_a_01}}]
    constraintWeights [1,1,1]
   }
   n18 = "Constraint - dyn_hair_back_l_01 (PPP)"(ComponentPose=@n17, Alpha=1, ConstraintWeights_0=0.4, ConstraintWeights_1=0.6, ConstraintWeights_2=0) {
    alphaBoolBlend {blendOption=Linear}
    boneToModify {boneName=dyn_hair_back_l_01}
    constraintSetup [{targetBone {boneName=neck_01}}, {targetBone {boneName=clavicle_l}}, {targetBone {boneName=dyn_hair_l_a_01}}]
    constraintWeights [1,1,1]
   }
   n19 = "Constraint - dyn_hair_back_r_01 (PPP)"(ComponentPose=@n18, Alpha=1, ConstraintWeights_0=0.4, ConstraintWeights_1=0.6, ConstraintWeights_2=0) {
    alphaBoolBlend {blendOption=Linear}
    boneToModify {boneName=dyn_hair_back_r_01}
    constraintSetup [{targetBone {boneName=neck_01}}, {targetBone {boneName=clavicle_r}}, {targetBone {boneName=dyn_hair_r_a_01}}]
    constraintWeights [1,1,1]
   }
   n20 = "Constraint - dyn_hair_back_m_01 (P)"(ComponentPose=@n19, Alpha=1, ConstraintWeights_0=1) {
    alphaBoolBlend {blendOption=Linear}
    boneToModify {boneName=dyn_hair_back_m_01}
    constraintSetup [{targetBone {boneName=neck_01}}]
    constraintWeights [1]
   }
   n21 = "Constraint - dyn_hair_back_r_02 (PPP)"(ComponentPose=@n20, Alpha=1, ConstraintWeights_0=0.3, ConstraintWeights_1=0.7, ConstraintWeights_2=0) {
    alphaBoolBlend {blendOption=Linear}
    boneToModify {boneName=dyn_hair_back_r_02}
    constraintSetup [{targetBone {boneName=spine_05}}, {targetBone {boneName=clavicle_r}}, {targetBone {boneName=dyn_hair_r_a_02}}]
    constraintWeights [1,1,1]
   }
   n22 = "Constraint - dyn_hair_back_l_02 (PPP)"(ComponentPose=@n21, Alpha=1, ConstraintWeights_0=0.4, ConstraintWeights_1=0.6, ConstraintWeights_2=0) {
    alphaBoolBlend {blendOption=Linear}
    boneToModify {boneName=dyn_hair_back_l_02}
    constraintSetup [{targetBone {boneName=spine_05}}, {targetBone {boneName=clavicle_l}}, {targetBone {boneName=dyn_hair_l_a_02}}]
    constraintWeights [1,1,1]
   }
   n23 = "Constraint - dyn_hair_back_m_02 (P)"(ComponentPose=@n22, Alpha=1, ConstraintWeights_0=1) {
    alphaBoolBlend {blendOption=Linear}
    boneToModify {boneName=dyn_hair_back_m_02}
    constraintSetup [{targetBone {boneName=spine_05}}]
    constraintWeights [1]
   }
   n24 = "Constraint - dyn_hair_back_l_03 (PP)"!(ComponentPose=@n23, Alpha=1, ConstraintWeights_0=0.7, ConstraintWeights_1:float=0.3) -> (Pose) {
    alphaBoolBlend {blendOption=Linear}
    boneToModify {boneName=dyn_hair_back_l_03}
    constraintSetup [{targetBone {boneName=clavicle_l}}, {targetBone {boneName=spine_05}}]
    constraintWeights [1,1]
   }
   n25 = "Transform (Modify) Bone - Bone: dyn_hair_back_m_03"(ComponentPose=@n24, Alpha=1, Translation="0,0,0", Rotation="0,0,5", Scale="1,1,1") {
    rotationMode=BMM_Additive
    translationMode=BMM_Additive
    translationSpace=BCS_BoneSpace
    alphaBoolBlend {blendOption=Linear}
    boneToModify {boneName=dyn_hair_back_m_03}
    rotation {roll=5}
   }
   n26 = "Transform (Modify) Bone - Bone: dyn_hair_back_l_03"(ComponentPose=@n25, Alpha=1, Translation="0,0,0", Rotation="0,0,5", Scale="1,1,1") {
    rotationMode=BMM_Additive
    translationMode=BMM_Additive
    translationSpace=BCS_BoneSpace
    alphaBoolBlend {blendOption=Linear}
    boneToModify {boneName=dyn_hair_back_l_03}
    rotation {roll=5}
   }
   n27 = "Transform (Modify) Bone - Bone: dyn_hair_back_r_b_03"(ComponentPose=@n26, Alpha=1, Translation="0,0,0", Rotation="0,0,5", Scale="1,1,1") {
    rotationMode=BMM_Additive
    translationMode=BMM_Additive
    translationSpace=BCS_BoneSpace
    alphaBoolBlend {blendOption=Linear}
    boneToModify {boneName=dyn_hair_back_r_b_03}
    rotation {roll=5}
   }
   n28 = "Transform (Modify) Bone - Bone: dyn_hair_back_r_03"(ComponentPose=@n27, Alpha=1, Translation="0,0,0", Rotation="0,0,5", Scale="1,1,1") {
    rotationMode=BMM_Additive
    translationMode=BMM_Additive
    translationSpace=BCS_BoneSpace
    alphaBoolBlend {blendOption=Linear}
    boneToModify {boneName=dyn_hair_back_r_03}
    rotation {roll=5}
   }
   n29 = Anim Dynamics - Bone: dyn_hair_back_m_02(ComponentPose=@n28, Alpha=1, ExternalForce="0,0,0") {
    angularDampingOverride=1
    angularSpringConstant=5500
    bAngularSpring=true
    bOverrideAngularDamping=true
    bUseGravityOverride=true
    gravityScale=0.35
    simulationSpace=BoneRelative
    alphaBoolBlend {blendOption=Linear}
    boundBone {boneName=dyn_hair_back_m_02}
    componentLinearAccScale=<1 1 1>
    componentLinearVelScale=<1 1 1>
    gravityOverride {z=-1500}
    physicsBodyDefinitions [{
      boundBone {boneName=dyn_hair_back_m_02}
      boxExtents=<16 16 16>
      constraintSetup {angularLimitsMax=<2 0 2>, angularLimitsMin=<-2 -40 -2>, angularTarget=<1 0 0>}
      localJointOffset=<1 0 0>
     }]
    relativeSpaceBone {boneName=root}
   }
   n30 = Anim Dynamics - Bone: dyn_hair_back_m_03(ComponentPose=@n29, Alpha=1, ExternalForce="0,0,0") {
    angularDampingOverride=1
    angularSpringConstant=5500
    bAngularSpring=true
    bOverrideAngularDamping=true
    bUseGravityOverride=true
    gravityScale=0.35
    simulationSpace=BoneRelative
    alphaBoolBlend {blendOption=Linear}
    boundBone {boneName=dyn_hair_back_m_03}
    componentLinearAccScale=<1 1 1>
    componentLinearVelScale=<1 1 1>
    gravityOverride {z=-1500}
    physicsBodyDefinitions [{
      boundBone {boneName=dyn_hair_back_m_03}
      boxExtents=<16 16 16>
      constraintSetup {angularLimitsMax=<2 0 2>, angularLimitsMin=<-2 -40 -2>, angularTarget=<1 0 0>}
      localJointOffset=<1 0 0>
     }]
    relativeSpaceBone {boneName=root}
   }
   n31 = Anim Dynamics - Bone: dyn_hair_back_m_04(ComponentPose=@n30, Alpha=1, ExternalForce="0,0,0") {
    angularDampingOverride=1
    angularSpringConstant=5500
    bAngularSpring=true
    bOverrideAngularDamping=true
    bUseGravityOverride=true
    gravityScale=0.35
    simulationSpace=BoneRelative
    alphaBoolBlend {blendOption=Linear}
    boundBone {boneName=dyn_hair_back_m_04}
    componentLinearAccScale=<1 1 1>
    componentLinearVelScale=<1 1 1>
    gravityOverride {z=-1500}
    physicsBodyDefinitions [{
      boundBone {boneName=dyn_hair_back_m_04}
      boxExtents=<16 16 16>
      constraintSetup {angularLimitsMax=<2 0 2>, angularLimitsMin=<-2 -40 -2>, angularTarget=<1 0 0>}
      localJointOffset=<1 0 0>
     }]
    relativeSpaceBone {boneName=root}
   }
   n32 = Anim Dynamics - Bone: dyn_hair_back_l_02(ComponentPose=@n31, Alpha=1, ExternalForce="0,0,0") {
    angularDampingOverride=1
    angularSpringConstant=5500
    bAngularSpring=true
    bOverrideAngularDamping=true
    bUseGravityOverride=true
    gravityScale=0.35
    simulationSpace=BoneRelative
    alphaBoolBlend {blendOption=Linear}
    boundBone {boneName=dyn_hair_back_l_02}
    componentLinearAccScale=<1 1 1>
    componentLinearVelScale=<1 1 1>
    gravityOverride {z=-1500}
    physicsBodyDefinitions [{
      boundBone {boneName=dyn_hair_back_l_02}
      boxExtents=<16 16 16>
      constraintSetup {angularLimitsMax=<2 0 2>, angularLimitsMin=<-2 -30 -2>, angularTarget=<1 0 0>}
      localJointOffset=<1 0 0>
     }]
    relativeSpaceBone {boneName=root}
   }
   n33 = Anim Dynamics - Bone: dyn_hair_back_l_03(ComponentPose=@n32, Alpha=1, ExternalForce="0,0,0") {
    angularDampingOverride=1
    angularSpringConstant=5500
    bAngularSpring=true
    bOverrideAngularDamping=true
    bUseGravityOverride=true
    gravityScale=0.35
    simulationSpace=BoneRelative
    alphaBoolBlend {blendOption=Linear}
    boundBone {boneName=dyn_hair_back_l_03}
    componentLinearAccScale=<1 1 1>
    componentLinearVelScale=<1 1 1>
    gravityOverride {z=-1500}
    physicsBodyDefinitions [{
      boundBone {boneName=dyn_hair_back_l_03}
      boxExtents=<16 16 16>
      constraintSetup {angularLimitsMax=<2 0 2>, angularLimitsMin=<-2 -40 -2>, angularTarget=<1 0 0>}
      localJointOffset=<1 0 0>
     }]
    relativeSpaceBone {boneName=root}
   }
   n34 = Anim Dynamics - Bone: dyn_hair_back_l_04(ComponentPose=@n33, Alpha=1, ExternalForce="0,0,0") {
    angularDampingOverride=1
    angularSpringConstant=5500
    bAngularSpring=true
    bOverrideAngularDamping=true
    bUseGravityOverride=true
    gravityScale=0.35
    simulationSpace=BoneRelative
    alphaBoolBlend {blendOption=Linear}
    boundBone {boneName=dyn_hair_back_l_04}
    componentLinearAccScale=<1 1 1>
    componentLinearVelScale=<1 1 1>
    gravityOverride {z=-1500}
    physicsBodyDefinitions [{
      boundBone {boneName=dyn_hair_back_l_04}
      boxExtents=<16 16 16>
      constraintSetup {angularLimitsMax=<2 0 2>, angularLimitsMin=<-2 -40 -2>, angularTarget=<1 0 0>}
      localJointOffset=<1 0 0>
     }]
    relativeSpaceBone {boneName=root}
   }
   n35 = Anim Dynamics - Bone: dyn_hair_back_r_02(ComponentPose=@n34, Alpha=1, ExternalForce="0,0,0") {
    angularDampingOverride=1
    angularSpringConstant=5500
    bAngularSpring=true
    bOverrideAngularDamping=true
    bUseGravityOverride=true
    gravityScale=0.35
    simulationSpace=BoneRelative
    alphaBoolBlend {blendOption=Linear}
    boundBone {boneName=dyn_hair_back_r_02}
    componentLinearAccScale=<1 1 1>
    componentLinearVelScale=<1 1 1>
    gravityOverride {z=-1500}
    physicsBodyDefinitions [{
      boundBone {boneName=dyn_hair_back_r_02}
      boxExtents=<16 16 16>
      constraintSetup {angularLimitsMax=<2 0 2>, angularLimitsMin=<-2 -40 -2>, angularTarget=<1 0 0>}
      localJointOffset=<1 0 0>
     }]
    relativeSpaceBone {boneName=root}
   }
   n36 = Anim Dynamics - Bone: dyn_hair_back_r_b_03(ComponentPose=@n35, Alpha=1, ExternalForce="0,0,0") {
    angularDampingOverride=1
    angularSpringConstant=5500
    bAngularSpring=true
    bOverrideAngularDamping=true
    bUseGravityOverride=true
    gravityScale=0.35
    simulationSpace=BoneRelative
    alphaBoolBlend {blendOption=Linear}
    boundBone {boneName=dyn_hair_back_r_b_03}
    componentLinearAccScale=<1 1 1>
    componentLinearVelScale=<1 1 1>
    gravityOverride {z=-1500}
    physicsBodyDefinitions [{
      boundBone {boneName=dyn_hair_back_r_b_03}
      boxExtents=<16 16 16>
      constraintSetup {angularLimitsMax=<2 -2 2>, angularLimitsMin=<-2 -40 -2>, angularTarget=<1 0 0>}
      localJointOffset=<1 0 0>
     }]
    relativeSpaceBone {boneName=root}
   }
   n37 = Anim Dynamics - Bone: dyn_hair_back_r_03(ComponentPose=@n36, Alpha=1, ExternalForce="0,0,0") {
    angularDampingOverride=1
    angularSpringConstant=5500
    bAngularSpring=true
    bOverrideAngularDamping=true
    bUseGravityOverride=true
    gravityScale=0.35
    simulationSpace=BoneRelative
    alphaBoolBlend {blendOption=Linear}
    boundBone {boneName=dyn_hair_back_r_03}
    componentLinearAccScale=<1 1 1>
    componentLinearVelScale=<1 1 1>
    gravityOverride {z=-1500}
    physicsBodyDefinitions [{
      boundBone {boneName=dyn_hair_back_r_03}
      boxExtents=<16 16 16>
      constraintSetup {angularLimitsMax=<2 -2 2>, angularLimitsMin=<-2 -40 -2>, angularTarget=<1 0 0>}
      localJointOffset=<1 0 0>
     }]
    relativeSpaceBone {boneName=root}
   }
   n38 = Anim Dynamics - Bone: dyn_hair_back_r_04(ComponentPose=@n37, Alpha=1, ExternalForce="0,0,0") {
    angularDampingOverride=1
    angularSpringConstant=5500
    bAngularSpring=true
    bOverrideAngularDamping=true
    bUseGravityOverride=true
    gravityScale=0.35
    simulationSpace=BoneRelative
    alphaBoolBlend {blendOption=Linear}
    boundBone {boneName=dyn_hair_back_r_04}
    componentLinearAccScale=<1 1 1>
    componentLinearVelScale=<1 1 1>
    gravityOverride {z=-1500}
    physicsBodyDefinitions [{
      boundBone {boneName=dyn_hair_back_r_04}
      boxExtents=<16 16 16>
      constraintSetup {angularLimitsMax=<2 -2 2>, angularLimitsMin=<-2 -40 -2>, angularTarget=<1 0 0>}
      localJointOffset=<1 0 0>
     }]
    relativeSpaceBone {boneName=root}
   }
   n39 = Component To Local(ComponentPose=@n38)
   n40 = Anim Dynamics - Bone: dyn_antena_r_01(
    ComponentPose=Local To Component(LocalPose=Blend Poses by bool(BlendPose_0=@n11, BlendPose_1=@n39) {blendType=Sinusoidal})
    Alpha=1
    ExternalForce="0,0,0"
   ) {
    angularDampingOverride=1
    angularSpringConstant=5500
    bAngularSpring=true
    bOverrideAngularDamping=true
    bUseGravityOverride=true
    gravityScale=0.35
    simulationSpace=BoneRelative
    alphaBoolBlend {blendOption=Linear}
    boundBone {boneName=dyn_antena_r_01}
    componentLinearAccScale=<1 1 1>
    componentLinearVelScale=<1 1 1>
    gravityOverride {z=-1500}
    physicsBodyDefinitions [{
      boundBone {boneName=dyn_antena_r_01}
      boxExtents=<18 18 18>
      constraintSetup {angularLimitsMax=<2 15 2>, angularLimitsMin=<-2 -15 -2>, angularTarget=<1 0 0>}
      localJointOffset=<1 0 0>
     }]
    relativeSpaceBone {boneName=root}
   }
   n41 = Anim Dynamics - Bone: dyn_antena_r_02(ComponentPose=@n40, Alpha=1, ExternalForce="0,0,0") {
    angularDampingOverride=1
    angularSpringConstant=5500
    bAngularSpring=true
    bOverrideAngularDamping=true
    bUseGravityOverride=true
    gravityScale=0.35
    simulationSpace=BoneRelative
    alphaBoolBlend {blendOption=Linear}
    boundBone {boneName=dyn_antena_r_02}
    componentLinearAccScale=<1 1 1>
    componentLinearVelScale=<1 1 1>
    gravityOverride {z=-1500}
    physicsBodyDefinitions [{
      boundBone {boneName=dyn_antena_r_02}
      boxExtents=<18 18 18>
      constraintSetup {angularLimitsMax=<2 15 2>, angularLimitsMin=<-2 -15 -2>, angularTarget=<1 0 0>}
      localJointOffset=<1 0 0>
     }]
    relativeSpaceBone {boneName=root}
   }
   n42 = Anim Dynamics - Bone: dyn_antena_r_03(ComponentPose=@n41, Alpha=1, ExternalForce="0,0,0") {
    angularDampingOverride=1
    angularSpringConstant=5500
    bAngularSpring=true
    bOverrideAngularDamping=true
    bUseGravityOverride=true
    gravityScale=0.35
    simulationSpace=BoneRelative
    alphaBoolBlend {blendOption=Linear}
    boundBone {boneName=dyn_antena_r_03}
    componentLinearAccScale=<1 1 1>
    componentLinearVelScale=<1 1 1>
    gravityOverride {z=-1500}
    physicsBodyDefinitions [{
      boundBone {boneName=dyn_antena_r_03}
      boxExtents=<18 18 18>
      constraintSetup {angularLimitsMax=<2 15 2>, angularLimitsMin=<-2 -15 -2>, angularTarget=<1 0 0>}
      localJointOffset=<1 0 0>
     }]
    relativeSpaceBone {boneName=root}
   }
   n43 = Anim Dynamics - Bone: dyn_antena_l_01(ComponentPose=@n42, Alpha=1, ExternalForce="0,0,0") {
    angularDampingOverride=1
    angularSpringConstant=5000
    bAngularSpring=true
    bOverrideAngularDamping=true
    bUseGravityOverride=true
    gravityScale=0.35
    simulationSpace=BoneRelative
    alphaBoolBlend {blendOption=Linear}
    boundBone {boneName=dyn_antena_l_01}
    componentLinearAccScale=<1 1 1>
    componentLinearVelScale=<1 1 1>
    gravityOverride {z=-1500}
    physicsBodyDefinitions [{
      boundBone {boneName=dyn_antena_l_01}
      boxExtents=<20 20 20>
      constraintSetup {angularLimitsMax=<2 15 2>, angularLimitsMin=<-2 -15 -2>, angularTarget=<1 0 0>}
      localJointOffset=<1 0 0>
     }]
    relativeSpaceBone {boneName=root}
   }
   n44 = Anim Dynamics - Bone: dyn_antena_l_02(ComponentPose=@n43, Alpha=1, ExternalForce="0,0,0") {
    angularDampingOverride=1
    angularSpringConstant=5000
    bAngularSpring=true
    bOverrideAngularDamping=true
    bUseGravityOverride=true
    gravityScale=0.35
    simulationSpace=BoneRelative
    alphaBoolBlend {blendOption=Linear}
    boundBone {boneName=dyn_antena_l_02}
    componentLinearAccScale=<1 1 1>
    componentLinearVelScale=<1 1 1>
    gravityOverride {z=-1500}
    physicsBodyDefinitions [{
      boundBone {boneName=dyn_antena_l_02}
      boxExtents=<20 20 20>
      constraintSetup {angularLimitsMax=<2 15 2>, angularLimitsMin=<-2 -15 -2>, angularTarget=<1 0 0>}
      localJointOffset=<1 0 0>
     }]
    relativeSpaceBone {boneName=root}
   }
   n45 = Anim Dynamics - Bone: dyn_antena_l_03(ComponentPose=@n44, Alpha=1, ExternalForce="0,0,0") {
    angularDampingOverride=1
    angularSpringConstant=5000
    bAngularSpring=true
    bOverrideAngularDamping=true
    bUseGravityOverride=true
    gravityScale=0.35
    simulationSpace=BoneRelative
    alphaBoolBlend {blendOption=Linear}
    boundBone {boneName=dyn_antena_l_03}
    componentLinearAccScale=<1 1 1>
    componentLinearVelScale=<1 1 1>
    gravityOverride {z=-1500}
    physicsBodyDefinitions [{
      boundBone {boneName=dyn_antena_l_03}
      boxExtents=<20 20 20>
      constraintSetup {angularLimitsMax=<2 15 2>, angularLimitsMin=<-2 -15 -2>, angularTarget=<1 0 0>}
      localJointOffset=<1 0 0>
     }]
    relativeSpaceBone {boneName=root}
   }
   Output Pose(Result=Component To Local(ComponentPose=@n45)) {layerGroup=None}
   n46 = Get Body_Mesh() -> (Body_Mesh:Skeletal Mesh Component Budgeted Object Reference)
   Save cached pose 'InitialPose'(Pose=Layered blend per bone(
     BasePose=Copy Pose From Mesh() {bUseAttachedParent=true}
     BlendPoses_0=Copy Pose From Mesh(SourceMeshComponent=@n46) {bUseAttachedParent=true}
    ) {layerSetup [{branchFilters [{blendDepth=1, boneName=clavicle_l}, {blendDepth=1, boneName=clavicle_r}]}]}) {cacheName=InitialPose}
   Reroute Node(InputPin:Wildcard) -> (OutputPin:Wildcard)
  ]}
 EventGraph {nodes [
   Event BlueprintUpdateAnimation()
   > n1 = Gravity Override State Function DEPRECATED(
    CharacterPartAnimInstance=Self-Reference()
    GameTimeInSeconds=GetGameTimeInSeconds()
    GravityOverrideParams=Get Gravity Override Params()
    DebugJointName=head
   )
   > Set Gravity Override(Gravity Override=@n1)
   > Set IsWearingBackbling(IsWearingBackbling:bool=CheckForBackbling()) -> (Output_Get:bool)
   > n2 = Branch(Condition=Get bIsInFrontEnd())
   > Set Hair RBAN FE Alpha(Hair RBAN FE Alpha:double=1) -> (Output_Get:double)
   > Set Hair RBAN InGame Alpha(Hair RBAN InGame Alpha:double=0) -> (Output_Get:double)
   > Set Hair FE Constraint Alpha(Hair FE Constraint Alpha:double=1) -> (Output_Get:double)
   > Set Hair InGame Constraint Alpha(Hair InGame Constraint Alpha:double=0) -> (Output_Get:double)
   Set Hair RBAN FE Alpha(execute=@n2.else, Hair RBAN FE Alpha=0)
   > Set Hair RBAN InGame Alpha(Hair RBAN InGame Alpha=1)
   > Set Hair FE Constraint Alpha(Hair FE Constraint Alpha=0)
   > Set Hair InGame Constraint Alpha(Hair InGame Constraint Alpha=1)
   n3 = Event OnCharacterAndPickaxeCustomizationLoaded!() -> (
    OutputDelegate
    then
    CustomBaseMesh:Skeletal Mesh Component Object Reference
    CustomParts:Array of Custom Character Part Object References
    CustomMeshParts:Array of Skeletal Mesh Component Budgeted Object References
   ) {is=title K2Node_Event}
   > Set Body_Mesh(Body_Mesh:Skeletal Mesh Component Budgeted Object Reference=Get Array Item with EFort Custom Part Type(Array=@n3.CustomMeshParts)) -> (Output_Get:Skeletal Mesh Component Budgeted Object Reference)
  ]}
]
variables [
 AnimDyn SimSpace {type=AnimPhysSimSpaceType}
 Body_Mesh {type=SkeletalMeshComponentBudgeted}
 Gravity Override {type=Vector}
 Gravity Override Params {type=GravityOverrideParamsStruct}
 Hair FE Constraint Alpha {type=number}
 Hair InGame Constraint Alpha {type=number}
 Hair RBAN FE Alpha {type=number}
 Hair RBAN InGame Alpha {type=number}
 isMelee {type=boolean}
 isSwitchMelee {type=boolean}
 IsWearingBackbling {type=boolean}
 meleeAlpha {type=number}
]
