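# Niagara system for the SameRabbit character body that shows haze sprites on the head, hidden by default through User.EnableVFX.
# Two near-identical GPU emitters each burst two camera-offset sprites that outlive their lifetime, sized by User.Size.
# Each update re-samples the skeletal mesh location, with the emitters differing only in bone/socket index, so the sprites follow the character.
# Alpha multiplies User.EnableVFX by a front-end/gameplay modifier, letting an external value drive visibility.
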NiagaraSystem /BRCosmetics/Characters/Player/Female/Medium/Bodies/F_MED_SameRabbit/VFX/NS_SameRabbit_Character_Reactivity
dependencies {usedDataInterfaces [], usedDynamicInputs [], usedModules [&SystemState]}
emitters {
 Haze_Head {
  simTarget=GPUComputeSim
  emitterUpdateScript [
   &EmitterState(Life Cycle Mode=::NewEnumerator1, Loop Behavior=::NewEnumerator1, Loop Duration Mode=::NewEnumerator1, UseLoopDelay=false)
   &SpawnBurst_Instantaneous(Spawn Count=2, Use Spawn Probability=false, UseLoopCountLimit=false)
  ]
  particleSpawnScript [
   &InitializeParticle(
    Color Mode=::NewEnumerator0
    Lifetime=0.1
    Sprite Size Mode=::NewEnumerator3
    Uniform Sprite Size=$User.Size
    UsePositionOffset=false
    Write Material Random=true
   )
   &CameraOffset(Camera Offset Amount=20)
  ]
  particleUpdateScript [
   &ParticleState(Kill Particles When Lifetime Has Elapsed=false, Let Infinitely Lived Particles Die When Emitter Deactivates=true)
   &SkeletalMeshLocation(
    Bone / Socket Index=1
    Bone Sampling Mode=::NewEnumerator4
    MaskSpawnGroup=false
    Sampled Position Offset=<1 0 0>
    Skeletal Mesh=$User.Skeletal Mesh
   )
   &ScaleColor(
    Scale Alpha=&Multiply_Float(A=&FN_FrontEndFloatModifier(Gameplay=1, Front End=1), B=$User.EnableVFX)
    Scale RGB=<2 2 1>
    ScaleA=true
    ScaleRGB=true
   )
  ]
  rendererClasses [&NiagaraSpriteRendererProperties]
  renderers [{rendererClass=&NiagaraSpriteRendererProperties}]
 }
 Haze_Head001 {
  simTarget=GPUComputeSim
  emitterUpdateScript [
   &EmitterState(Life Cycle Mode=::NewEnumerator1, Loop Behavior=::NewEnumerator1, Loop Duration Mode=::NewEnumerator1, UseLoopDelay=false)
   &SpawnBurst_Instantaneous(Spawn Count=2, Use Spawn Probability=false, UseLoopCountLimit=false)
  ]
  particleSpawnScript [
   &InitializeParticle(
    Color Mode=::NewEnumerator0
    Lifetime=0.1
    Sprite Size Mode=::NewEnumerator3
    Uniform Sprite Size=$User.Size
    UsePositionOffset=false
    Write Material Random=true
   )
   &CameraOffset(Camera Offset Amount=20)
  ]
  particleUpdateScript [
   &ParticleState(Kill Particles When Lifetime Has Elapsed=false, Let Infinitely Lived Particles Die When Emitter Deactivates=true)
   &SkeletalMeshLocation(Bone Sampling Mode=::NewEnumerator4, MaskSpawnGroup=false, Sampled Position Offset=<1 0 0>, Skeletal Mesh=$User.Skeletal Mesh)
   &ScaleColor(
    Scale Alpha=&Multiply_Float(A=&FN_FrontEndFloatModifier(Gameplay=1, Front End=1), B=$User.EnableVFX)
    Scale RGB=<2 2 1>
    ScaleA=true
    ScaleRGB=true
   )
  ]
  rendererClasses [&NiagaraSpriteRendererProperties]
  renderers [{rendererClass=&NiagaraSpriteRendererProperties}]
 }
}
summary {
 systemName=NS_SameRabbit_Character_Reactivity
 emitters [
  {emitterName=Haze_Head001, simTarget=GPUComputeSim, rendererClasses [&NiagaraSpriteRendererProperties]}
  {emitterName=Haze_Head, simTarget=GPUComputeSim, rendererClasses [&NiagaraSpriteRendererProperties]}
 ]
 userVariables [
  User.bIsFrontEnd {type=NiagaraBool}
  User.bIsFrontEndPreview {type=NiagaraBool}
  User.EnableVFX {type=NiagaraFloat, defaultValue=0, description=""}
  User.Size {type=NiagaraFloat, defaultValue=6.5, description=""}
  User.Skeletal Mesh {
   type=NiagaraDataInterfaceSkeletalMesh
   defaultValue {
    struct=NiagaraExt_VariableValue_DataInterface
    value {dataInterface=&self:NiagaraDataInterfaceSkeletalMesh_0}
   }
  }
 ]
}
